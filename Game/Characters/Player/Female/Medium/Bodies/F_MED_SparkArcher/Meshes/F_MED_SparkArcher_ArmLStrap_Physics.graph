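# Physics asset for the left upper-arm vest strap on the medium female SparkArcher player body.
# Three capsule bodies cover the left upper arm and the two dyn strap bones.
# Constraints form a chain from the arm to the first strap bone to the second, with limited swing and twist at each joint.
# Twist is tighter at the arm attachment than between the strap segments.
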PhysicsAsset /Game/Characters/Player/Female/Medium/Bodies/F_MED_SparkArcher/Meshes/F_MED_SparkArcher_ArmLStrap_Physics
bodies {
 dyn_upperarm_vest_strap_l_01 [{shapeType=Capsule, length=4, radius=2, shapeName=dyn_upperarm_vest_strap_l_01_capsule, center=<3.6942 -0.6427 1.6031>, rotation=<rot 90 0 0>}]
 dyn_upperarm_vest_strap_l_02 [{
   shapeType=Capsule
   length=10.0474
   radius=2
   shapeName=dyn_upperarm_vest_strap_l_02_capsule
   center=<3.3066 -0.2672 1.4446>
   rotation=<rot 86.1443 0.7698 5.5651>
  }]
 upperarm_l [{
   shapeType=Capsule
   length=18.8821
   radius=5.9871
   shapeName=upperarm_l_capsule
   center=<13.0316 -0.2022 -2.7862>
   rotation=<rot -90 180 176.6153>
  }]
}
constraints [
 {bone1Name=dyn_upperarm_vest_strap_l_01, bone2Name=upperarm_l, swing1LimitDegrees=45, swing2LimitDegrees=45, twistLimitDegrees=10}
 {
  bone1Name=dyn_upperarm_vest_strap_l_02
  bone2Name=dyn_upperarm_vest_strap_l_01
  swing1LimitDegrees=45
  swing2LimitDegrees=45
  twistLimitDegrees=20
 }
]
