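# Physics asset for the right upper-arm vest strap on the medium female Spark Archer player body, for simulating the strap's motion on the arm.
# Capsule bodies cover the upper arm and two dynamic strap bones.
# Constraints chain the upper arm to the first strap bone and on to the second, with moderate swing and tighter twist limits that loosen toward the tip.
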PhysicsAsset /Game/Characters/Player/Female/Medium/Bodies/F_MED_SparkArcher/Meshes/F_MED_SparkArcher_ArmRStrap_Physics
bodies {
 dyn_upperarm_vest_strap_r_01 [{
   shapeType=Capsule
   length=3.2554
   radius=2
   shapeName=dyn_upperarm_vest_strap_r_01_capsule
   center=<-2.3378 0 -1.3985>
   rotation=<rot 90 0 0>
  }]
 dyn_upperarm_vest_strap_r_02 [{
   shapeType=Capsule
   length=6
   radius=2
   shapeName=dyn_upperarm_vest_strap_r_02_capsule
   center=<-4.2214 0.649 -0.8529>
   rotation=<rot 84.1765 -3.827 -0.0652>
  }]
 upperarm_r [{
   shapeType=Capsule
   length=19.9831
   radius=4.5957
   shapeName=upperarm_r_capsule
   center=<-15.7819 0.9618 1.6878>
   rotation=<rot -88.751 -11.1824 12.515>
  }]
}
constraints [
 {bone1Name=dyn_upperarm_vest_strap_r_01, bone2Name=upperarm_r, swing1LimitDegrees=45, swing2LimitDegrees=45, twistLimitDegrees=10}
 {
  bone1Name=dyn_upperarm_vest_strap_r_02
  bone2Name=dyn_upperarm_vest_strap_r_01
  swing1LimitDegrees=45
  swing2LimitDegrees=45
  twistLimitDegrees=20
 }
]
