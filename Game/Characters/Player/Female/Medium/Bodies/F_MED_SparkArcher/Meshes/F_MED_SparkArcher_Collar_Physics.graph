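# Physics asset for the SparkArcher female medium body's collar mesh, defining collision and joint limits for the dynamic vest collar and front strap bones.
# Short constraint chains link collar pieces to tiny base capsules, with straps continuing off the side collars and narrow swing and twist limits.
# Larger spine_04 and upper-arm capsules supply torso and arm collision shapes.
# Setup is asymmetric: right collar pieces use boxes, and the right front strap attaches to spine_04, the left to its own base with zero-degree limits.
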PhysicsAsset /Game/Characters/Player/Female/Medium/Bodies/F_MED_SparkArcher/Meshes/F_MED_SparkArcher_Collar_Physics
bodies {
 base_dyn_vest_strap_frnt_l [{shapeType=Capsule, length=0.505, radius=0.505, shapeName=base_dyn_vest_strap_frnt_l_capsule, rotation=<rot 0 -90 0>}]
 dyn_vest_collar_frnt_l_01 [{shapeType=Capsule, length=20.0008, radius=4.4802, shapeName=dyn_vest_collar_frnt_l_01_capsule, rotation=<rot 0 9.1393 89.9999>}]
 dyn_vest_collar_frnt_l_base [{shapeType=Capsule, length=0.505, radius=0.505, shapeName=dyn_vest_collar_frnt_l_base_capsule, rotation=<rot 0 -90 0>}]
 dyn_vest_collar_frnt_r_01 [{
   shapeType=Capsule
   length=18.1592
   radius=4.6343
   shapeName=dyn_vest_collar_frnt_r_01_capsule
   center=<-2.249 0.1377 -4>
   rotation=<rot 0.2548 7.9025 -91.8349>
  }]
 dyn_vest_collar_frnt_r_base [{shapeType=Capsule, length=0.505, radius=0.505, shapeName=dyn_vest_collar_frnt_r_base_capsule}]
 dyn_vest_collar_l_01 [{
   shapeType=Capsule
   length=17.9085
   radius=2.0315
   shapeName=dyn_vest_collar_l_01_capsule
   center=<5.8998 0.1879 1.044>
   rotation=<rot 0 -172.0023 -86.5482>
  }]
 dyn_vest_collar_l_base [{shapeType=Capsule, length=0.505, radius=0.505, shapeName=dyn_vest_collar_l_base_capsule, rotation=<rot 0 -90 0>}]
 dyn_vest_collar_mid_01 [{shapeType=Capsule, length=6, radius=4, shapeName=dyn_vest_collar_mid_01_capsule}]
 dyn_vest_collar_mid_02 [{shapeType=Capsule, length=10.1884, radius=4, shapeName=dyn_vest_collar_mid_02_capsule}]
 dyn_vest_collar_mid_base [{shapeType=Capsule, length=0.505, radius=0.505, shapeName=dyn_vest_collar_mid_base_capsule, rotation=<rot 0 -90 0>}]
 dyn_vest_collar_r_01 [{shapeType=Box, extentX=11.2148, extentY=12.4617, extentZ=1.1438, shapeName=dyn_vest_collar_r_01_box, center=<-4.8235 -0.154 -0.7939>}]
 dyn_vest_collar_r_base [{shapeType=Capsule, length=0.505, radius=0.505, shapeName=dyn_vest_collar_r_base_capsule, rotation=<rot 0 -90 0>}]
 dyn_vest_collar_strap_l_01 [{
   shapeType=Capsule
   length=0.9207
   radius=1.7477
   shapeName=dyn_vest_collar_strap_l_01_capsule
   center=<1.8077 -0.0612 -0.1111>
   rotation=<rot 90 -134.4317 -131.6837>
  }]
 dyn_vest_collar_strap_r_01 [{shapeType=Box, extentX=3.1322, extentY=2.9649, extentZ=0.5079, shapeName=dyn_vest_collar_strap_r_01_box, center=<-1.6093 -0.5084 0>}]
 dyn_vest_strap_frnt_l_01 [{
   shapeType=Capsule
   length=3.1705
   radius=0.9455
   shapeName=dyn_vest_strap_frnt_l_01_capsule
   center=<2.7334 -0.0946 0.0777>
   rotation=<rot 0 -88.154 -89.5292>
  }]
 dyn_vest_strap_frnt_l_02 [{
   shapeType=Capsule
   length=6.8834
   radius=1.1166
   shapeName=dyn_vest_strap_frnt_l_02_capsule
   center=<3.8827 0 -0.4199>
   rotation=<rot -83.828 179.9998 -179.9997>
  }]
 dyn_vest_strap_frnt_r_01 [{
   shapeType=Capsule
   length=4.4332
   radius=1.1864
   shapeName=dyn_vest_strap_frnt_r_01_capsule
   center=<-2.7337 0.0947 -0.0776>
   rotation=<rot 0 -88.154 -89.5292>
  }]
 dyn_vest_strap_frnt_r_02 [{
   shapeType=Capsule
   length=6.8834
   radius=1.1166
   shapeName=dyn_vest_strap_frnt_r_02_capsule
   center=<-4.0822 0 0.4415>
   rotation=<rot -83.828 179.9998 -179.9997>
  }]
 spine_04 [
  {shapeType=Capsule, length=18.9242, radius=11.9514, shapeName=spine_04_capsule, center=<1.4619 -2.7313 0>}
  {shapeType=Capsule, length=22.274, radius=8.3645, shapeName=spine_04_capsule, center=<-4.3294 3.3927 4.6767>}
 ]
 upperarm_l [{shapeType=Capsule, length=17, radius=5, shapeName=upperarm_l_capsule, center=<11.0795 0 0>, rotation=<rot -90 0.0099 -0.0099>}]
 upperarm_r [{shapeType=Capsule, length=17, radius=4, shapeName=upperarm_r_capsule, center=<-10.8221 0 0>, rotation=<rot -90 -180 180>}]
}
constraints [
 {bone1Name=dyn_vest_collar_frnt_l_01, bone2Name=dyn_vest_collar_frnt_l_base, swing1LimitDegrees=10, swing2LimitDegrees=5, twistLimitDegrees=0}
 {bone1Name=dyn_vest_collar_strap_l_01, bone2Name=dyn_vest_collar_l_01, swing1LimitDegrees=30, swing2LimitDegrees=5, twistLimitDegrees=10}
 {
  bone1Name=dyn_vest_collar_frnt_r_01
  bone2Name=dyn_vest_collar_frnt_r_base
  swing1LimitDegrees=1.44
  swing2LimitDegrees=10
  twistLimitDegrees=0
 }
 {bone1Name=dyn_vest_collar_r_01, bone2Name=dyn_vest_collar_r_base, swing1LimitDegrees=5, swing2LimitDegrees=4.16, twistLimitDegrees=0}
 {
  bone1Name=dyn_vest_collar_strap_r_01
  bone2Name=dyn_vest_collar_r_01
  swing1LimitDegrees=15.84
  swing2LimitDegrees=2.64
  twistLimitDegrees=10
 }
 {bone1Name=dyn_vest_collar_mid_01, bone2Name=dyn_vest_collar_mid_base, swing1LimitDegrees=30, swing2LimitDegrees=20, twistLimitDegrees=5}
 {bone1Name=dyn_vest_collar_mid_02, bone2Name=dyn_vest_collar_mid_01, swing1LimitDegrees=20, swing2LimitDegrees=20, twistLimitDegrees=5}
 {bone1Name=dyn_vest_strap_frnt_l_02, bone2Name=dyn_vest_strap_frnt_l_01, swing1LimitDegrees=5, swing2LimitDegrees=5, twistLimitDegrees=2}
 {bone1Name=dyn_vest_strap_frnt_r_02, bone2Name=dyn_vest_strap_frnt_r_01, swing1LimitDegrees=5, swing2LimitDegrees=10, twistLimitDegrees=2}
 {bone1Name=dyn_vest_collar_l_01, bone2Name=dyn_vest_collar_l_base, swing1LimitDegrees=5, swing2LimitDegrees=2, twistLimitDegrees=0}
 {bone1Name=dyn_vest_strap_frnt_r_01, bone2Name=spine_04, swing1LimitDegrees=5, swing2LimitDegrees=5, twistLimitDegrees=1}
 {bone1Name=dyn_vest_strap_frnt_l_01, bone2Name=base_dyn_vest_strap_frnt_l, swing1LimitDegrees=0, swing2LimitDegrees=0, twistLimitDegrees=0}
]
