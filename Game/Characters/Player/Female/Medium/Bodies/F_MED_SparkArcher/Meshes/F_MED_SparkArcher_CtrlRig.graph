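# Control Rig for the F_MED_SparkArcher female body that applies pose-driven corrective offsets to vest, collar, foot lace and knee bones.
# Spherical pose readers and remapped relative transforms weight ModifyTransforms nodes, some Kalman-smoothed, alongside collar-support parent constraints.
# A shared DeformRig function handles standard deform bones.
# Sliding adds vest offsets, while AltSkinCorrective shifts with pelvis height or interpolated tactical sprint when RigidBody is disabled.
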ControlRigBlueprint /Game/Characters/Player/Female/Medium/Bodies/F_MED_SparkArcher/Meshes/F_MED_SparkArcher_CtrlRig
graphs [
 RigVMFunctionLibrary {links [], nodes []}
 RigVMFunction_Sequence_1_1_ContainedGraph {nodes [
   n1 = RigVMFunctionEntryNode()
   n2 = &RigVMFunction_Sequence(ExecuteContext=@n1.ExecuteContext)
   n3 = &RigVMFunction_Sequence(ExecuteContext=@n2.B)
   RigVMFunctionReturnNode(A=@n2.A, B=@n3.A, C=@n3.B)
  ]}
 RigVMFunction_Sequence_1_ContainedGraph {nodes [
   n1 = RigVMFunctionEntryNode()
   RigVMFunction_Sequence = &RigVMFunction_Sequence(ExecuteContext=@n1.ExecuteContext)
   n2 = &RigVMFunction_Sequence(ExecuteContext=@RigVMFunction_Sequence.B)
   n3 = &RigVMFunction_Sequence(ExecuteContext=@n2.B)
   n4 = &RigVMFunction_Sequence(ExecuteContext=@n3.B)
   RigVMFunctionReturnNode(A=@RigVMFunction_Sequence.A, B=@n2.A, C=@n3.A, D=@n4.A, E=@n4.B)
  ]}
 RigVMModel {nodes [
   RigUnit_BeginExecution = &RigUnit_BeginExecution()
   RigVMFunction_Sequence_1 = RigVMAggregateNode(ExecuteContext=@RigUnit_BeginExecution.ExecutePin)
   RigVMFunction_ControlFlowBranch = &RigVMFunction_ControlFlowBranch(ExecuteContext=@RigVMFunction_Sequence_1.A, Condition=RigVMVariableNode() {variable=RigidBody Enabled}.Value)
   &RigUnit_ModifyTransforms(
    ItemToModify=((Item=(Type=Bone,Name="AltSkinCorrective"),Transform=(Rotation=(X=0,Y=0,Z=0,W=1),Translation=(X=-6.3903,Y=-15,Z=0),Scale3D=(X=1,Y=1,Z=1))))
    WeightMinimum=0
    Weight=&RigVMFunction_MathDoubleRemap(
     SourceMaximum=42
     SourceMinimum=80
     TargetMaximum=1
     TargetMinimum=0
     Value=&RigUnit_GetRelativeTransformForItem(Child=(Type=Bone,Name="pelvis"), Parent=(Type=Bone,Name="root"), bChildInitial=False, bParentInitial=False).RelativeTransform.Translation.Z
    ).Result
    ExecutePin=@RigVMFunction_ControlFlowBranch.False
   )
   RigVMFunction_Sequence = RigVMAggregateNode(ExecuteContext=@RigVMFunction_Sequence_1.B)
   n1 = &RigUnit_SphericalPoseReader(
    ActiveRegionScaleFactors=(PositiveWidth=1,NegativeWidth=1,PositiveHeight=1,NegativeHeight=1)
    ActiveRegionSize=0.1
    Debug=(bDrawDebug=false,bDraw2D=False,bDrawLocalAxes=False,DebugScale=25,DebugSegments=20,DebugThickness=0.25)
    DriverAxis=(X=1,Y=0,Z=0)
    DriverItem=(Type=Bone,Name="upperarm_l")
    FalloffRegionScaleFactors=(PositiveWidth=1,NegativeWidth=1,PositiveHeight=1,NegativeHeight=1)
    FalloffSize=0.4
    FlipHeightScaling=False
    FlipWidthScaling=False
    OptionalParentItem=(Type=Bone,Name="None")
    RotationOffset=(X=65,Y=0,Z=-40)
    ExecutePin=@RigVMFunction_Sequence.A
   )
   > &RigUnit_ModifyTransforms(
    ItemToModify="((Item=(Type=Bone,Name=\"None\"),Transform=(Rotation=(X=0,Y=0,Z=0,W=1),Translation=(X=0,Y=0,Z=0),Scale3D=(X=1,Y=1,Z=1))),(Item=(Type=Bone,Name=\"dyn_vest_collar_frnt_l_base\"),Transform=(Rotation=(X=0,Y=0,Z=0,W=1),Translation=(X=-1,Y=0,Z=0),Scale3D=(X=1,Y=1,Z= ...(truncated)"
    WeightMinimum=0
    ItemToModify.1.Transform.Rotation=&RigVMFunction_MathQuaternionFromEuler(Euler=(X=0,Y=-20,Z=0)).Result
    ItemToModify.2.Transform.Rotation=&RigVMFunction_MathQuaternionFromEuler(Euler=(X=15,Y=-15,Z=15)).Result
    Weight=@n1.OutputParam
   )
   > n2 = &RigUnit_SphericalPoseReader(
    ActiveRegionScaleFactors=(PositiveWidth=1,NegativeWidth=1,PositiveHeight=1,NegativeHeight=1)
    ActiveRegionSize=0.1
    Debug=(bDrawDebug=true,bDraw2D=False,bDrawLocalAxes=False,DebugScale=25,DebugSegments=20,DebugThickness=0.25)
    DriverAxis=(X=-1,Y=0,Z=0)
    DriverItem=(Type=Bone,Name="upperarm_r")
    FalloffRegionScaleFactors=(PositiveWidth=0.5,NegativeWidth=1,PositiveHeight=0.5,NegativeHeight=1)
    FalloffSize=0.5
    FlipHeightScaling=False
    FlipWidthScaling=False
    OptionalParentItem=(Type=Bone,Name="None")
    RotationOffset=(X=-70,Y=0,Z=30)
   )
   > &RigUnit_ModifyTransforms(
    ItemToModify="((Item=(Type=Bone,Name=\"dyn_vest_collar_r_base\"),Transform=(Rotation=(X=0.4224,Y=0.0738,Z=0.3929,W=0.8135),Translation=(X=0.1647,Y=2.0523,Z=0.7334),Scale3D=(X=1,Y=1,Z=1))),(Item=(Type=Bone,Name=\"dyn_vest_collar_frnt_r_base\"),Transform=(Rotation=(X=-0.0363, ...(truncated)"
    WeightMinimum=0
    Weight=&RigVMFunction_KalmanFloat(BufferSize=5, Value=@n2.OutputParam) {template="KalmanFilter::Execute(in Value,in BufferSize,out Result)"}.Result
   )
   > n3 = &RigUnit_SphericalPoseReader(
    ActiveRegionScaleFactors=(PositiveWidth=1,NegativeWidth=1,PositiveHeight=1,NegativeHeight=1)
    ActiveRegionSize=0.1
    Debug=(bDrawDebug=false,bDraw2D=False,bDrawLocalAxes=False,DebugScale=25,DebugSegments=20,DebugThickness=0.25)
    DriverAxis=(X=-1,Y=0,Z=0)
    DriverItem=(Type=Bone,Name="upperarm_r")
    FalloffRegionScaleFactors=(PositiveWidth=1,NegativeWidth=1,PositiveHeight=1,NegativeHeight=1)
    FalloffSize=0.22
    FlipHeightScaling=False
    FlipWidthScaling=False
    OptionalParentItem=(Type=Bone,Name="None")
    RotationOffset=(X=-120,Y=20,Z=30)
   )
   RigUnit_GetRelativeTransformForItem = &RigUnit_GetRelativeTransformForItem(Child=(Type=Bone,Name="pelvis"), Parent=(Type=Bone,Name="root"), bChildInitial=False, bParentInitial=False)
   Remap = &RigVMFunction_MathDoubleRemap(
    SourceMaximum=70
    SourceMinimum=85
    TargetMaximum=1
    TargetMinimum=0
    Value=@RigUnit_GetRelativeTransformForItem.RelativeTransform.Translation.Z
   )
   > &RigUnit_ModifyTransforms(
    ItemToModify="((Item=(Type=Bone,Name=\"dyn_vest_collar_strap_r_base\"),Transform=(Rotation=(X=0,Y=0,Z=0,W=1),Translation=(X=0,Y=0,Z=0.164),Scale3D=(X=1,Y=1,Z=1))),(Item=(Type=Bone,Name=\"dyn_vest_collar_r_base\"),Transform=(Rotation=(X=-0.0299,Y=-0.6446,Z=0.1152,W=0.7552),T ...(truncated)"
    WeightMinimum=0
    Weight=&RigVMFunction_MathDoubleMul(B=@Remap.Result, A=@n3.OutputParam).Result
   )
   > &RigUnit_ModifyTransforms(
    ItemToModify=((Item=(Type=Bone,Name="dyn_vest_collar_r_base"),Transform=(Rotation=(X=0,Y=-0.291,Z=0,W=0.9567),Translation=(X=0,Y=0,Z=0),Scale3D=(X=1,Y=1,Z=1))))
    WeightMinimum=0
    Weight=&RigVMFunction_MathDoubleMul(
     B=&RigVMFunction_MathDoubleRemap(SourceMaximum=0, SourceMinimum=1, TargetMaximum=1, TargetMinimum=0, Value=@Remap.Result).Result
     A=@n3.OutputParam
    ).Result
   )
   > &RigUnit_ParentConstraint(
    AdvancedSettings=(InterpolationType=Average,RotationOrderForFilter=XZY)
    Child=(Type=Bone,Name="dyn_vest_collar_frnt_r_support")
    Filter=(TranslationFilter=(bX=True,bY=True,bZ=True),RotationFilter=(bX=false,bY=false,bZ=false),ScaleFilter=(bX=false,bY=false,bZ=false))
    Parents=((Item=(Type=Bone,Name="dyn_vest_collar_frnt_r_base"),Weight=1))
    Weight=1
    bMaintainOffset=true
   )
   RigVMFunction_MathQuaternionToEuler = &RigVMFunction_MathQuaternionToEuler(
    RotationOrder=ZYX
    Value=&RigUnit_GetRelativeTransformForItem(Child=(Type=Bone,Name="spine_05"), Parent=(Type=Bone,Name="spine_01"), bChildInitial=False, bParentInitial=False).RelativeTransform.Rotation
   )
   > &RigUnit_ModifyTransforms(
    ItemToModify=((Item=(Type=Bone,Name="dyn_vest_collar_frnt_r_support"),Transform=(Rotation=(X=0,Y=0,Z=0,W=1),Translation=(X=2.3555,Y=0,Z=0.2),Scale3D=(X=1,Y=1,Z=1))))
    WeightMinimum=0
    Weight=&RigVMFunction_MathDoubleRemap(SourceMaximum=60, SourceMinimum=80, TargetMaximum=1, TargetMinimum=0, Value=@RigVMFunction_MathQuaternionToEuler.Result.Z).Result
   )
   Multiply = &RigVMFunction_MathDoubleMul(A=1, B=@n2.OutputParam)
   KalmanFilter = &RigVMFunction_KalmanFloat(BufferSize=5, Value=@Multiply.Result) {template="KalmanFilter::Execute(in Value,in BufferSize,out Result)"}
   > &RigUnit_ModifyTransforms(
    ItemToModify=((Item=(Type=Bone,Name="dyn_vest_collar_r_base"),Transform=(Rotation=(X=-0.0248,Y=-0.1038,Z=-0.045,W=0.9933),Translation=(X=-0.1631,Y=-3.3544,Z=2.0638),Scale3D=(X=1,Y=1,Z=1))))
    WeightMinimum=0
    Weight=@KalmanFilter.Result
   )
   > n4 = &RigUnit_SphericalPoseReader(
    ActiveRegionScaleFactors=(PositiveWidth=1,NegativeWidth=1,PositiveHeight=1,NegativeHeight=1)
    ActiveRegionSize=0.1
    Debug=(bDrawDebug=false,bDraw2D=False,bDrawLocalAxes=False,DebugScale=25,DebugSegments=20,DebugThickness=0.25)
    DriverAxis=(X=1,Y=0,Z=0)
    DriverItem=(Type=Bone,Name="head")
    FalloffRegionScaleFactors=(PositiveWidth=1,NegativeWidth=1,PositiveHeight=1,NegativeHeight=1)
    FalloffSize=0.4
    FlipHeightScaling=False
    FlipWidthScaling=False
    OptionalParentItem=(Type=Bone,Name="None")
    RotationOffset=(X=0,Y=-60,Z=-60)
   )
   > &RigUnit_ModifyTransforms(
    ItemToModify=((Item=(Type=Bone,Name="dyn_vest_collar_mid_base"),Transform=(Rotation=(X=0,Y=0,Z=0,W=1),Translation=(X=-2,Y=0,Z=4),Scale3D=(X=1,Y=1,Z=1))))
    WeightMinimum=0
    ItemToModify.0.Transform.Rotation=&RigVMFunction_MathQuaternionFromEuler(Euler=(X=10,Y=-20,Z=-30)).Result
    Weight=@n4.OutputParam
   )
   FN_DeformRig_FNC = RigVMFunctionReferenceNode(
    DeformRig_ Curve="(deformRig_l_upperarm_up_alpha_72_#1=False,deformRig_r_upperarm_up_alpha_73_#2=False,deformRig_l_upperarm_fr_alpha_74_#3=False,deformRig_r_upperarm_fr_alpha_75_CB51B8 ...(truncated)"
    DeformRig_Bool="(upperarm_1_#4=True,elbow_3_#5=True,wrist_7_#6=True,groin_8_#7=True,glute_17_#8=True,knee_9_40C408324CA261 ...(truncated)"
    DeformRig_Debug=(upperarm_debug_18_#4=False,wrist_debug_20_#6=False,lat_debug_26_#9=False,groinGlute_debug_27_#7=False)
    DeformRig_Value="(upperarm_up_18_#10=(Rotation=(X=0,Y=0,Z=0,W=1),Translation=(X=0,Y=0,Z=8),Scale3D=(X=0.7,Y=1,Z=1)),upperarm_fwd_fr_19_#11=(Rotation=(X=0,Y=0,Z=0,W=1),Translation=(X=0,Y=10,Z=0),Scale3D=(X=0.5,Y=1,Z= ...(truncated)"
    ExecuteContext=@RigVMFunction_Sequence.B
   )
   n5 = &RigUnit_SphericalPoseReader(
    ActiveRegionScaleFactors=(PositiveWidth=1,NegativeWidth=1,PositiveHeight=1,NegativeHeight=1)
    ActiveRegionSize=0.2
    Debug=(bDrawDebug=true,bDraw2D=False,bDrawLocalAxes=False,DebugScale=25,DebugSegments=20,DebugThickness=0.25)
    DriverAxis=(X=1,Y=0,Z=0)
    DriverItem=(Type=Bone,Name="upperarm_l")
    FalloffRegionScaleFactors=(PositiveWidth=1,NegativeWidth=1,PositiveHeight=1,NegativeHeight=1)
    FalloffSize=0.3
    FlipHeightScaling=False
    FlipWidthScaling=False
    OptionalParentItem=(Type=Bone,Name="None")
    RotationOffset=(X=0,Y=-50,Z=0)
    ExecutePin=@RigVMFunction_Sequence.C
   )
   > &RigUnit_ModifyTransforms(
    ItemToModify="((Item=(Type=Bone,Name=\"dyn_vest_collar_frnt_l_base\"),Transform=(Rotation=(X=0,Y=0,Z=0,W=1),Translation=(X=-2,Y=0,Z=0),Scale3D=(X=1,Y=1,Z=1))),(Item=(Type=Bone,Name=\"dyn_vest_collar_l_base\"),Transform=(Rotation=(X=0,Y=0,Z=0,W=1),Translation=(X=0,Y=2,Z=0),S ...(truncated)"
    WeightMinimum=0
    ItemToModify.1.Transform.Rotation=&RigVMFunction_MathQuaternionFromEuler(Euler=(X=10,Y=10,Z=15)).Result
    ItemToModify.0.Transform.Rotation=&RigVMFunction_MathQuaternionFromEuler(Euler=(X=0,Y=0,Z=0)).Result
    Weight=@n5.OutputParam
   )
   > n6 = &RigUnit_SphericalPoseReader(
    ActiveRegionScaleFactors=(PositiveWidth=1,NegativeWidth=1,PositiveHeight=1,NegativeHeight=1)
    ActiveRegionSize=0.2
    Debug=(bDrawDebug=false,bDraw2D=False,bDrawLocalAxes=False,DebugScale=25,DebugSegments=20,DebugThickness=0.25)
    DriverAxis=(X=-1,Y=0,Z=0)
    DriverItem=(Type=Bone,Name="upperarm_r")
    FalloffRegionScaleFactors=(PositiveWidth=1,NegativeWidth=1,PositiveHeight=1,NegativeHeight=1)
    FalloffSize=0.3
    FlipHeightScaling=False
    FlipWidthScaling=False
    OptionalParentItem=(Type=Bone,Name="None")
    RotationOffset=(X=0,Y=135,Z=0)
   )
   > &RigUnit_ModifyTransforms(
    ItemToModify="((Item=(Type=Bone,Name=\"dyn_vest_collar_frnt_r_base\"),Transform=(Rotation=(X=0,Y=0,Z=0,W=1),Translation=(X=2,Y=0,Z=0),Scale3D=(X=1,Y=1,Z=1))),(Item=(Type=Bone,Name=\"dyn_vest_collar_r_base\"),Transform=(Rotation=(X=0,Y=0,Z=0,W=1),Translation=(X=0,Y=0,Z=0),Sc ...(truncated)"
    WeightMinimum=0
    ItemToModify.1.Transform.Rotation=&RigVMFunction_MathQuaternionFromEuler(Euler=(X=10,Y=10,Z=30)).Result
    ItemToModify.0.Transform.Rotation=&RigVMFunction_MathQuaternionFromEuler(Euler=(X=0,Y=0,Z=0)).Result
    Weight=@n6.OutputParam
   )
   > &RigUnit_ParentConstraint(
    AdvancedSettings=(InterpolationType=Average,RotationOrderForFilter=XZY)
    Child=(Type=Bone,Name="dyn_vest_collar_frnt_l_support")
    Filter=(TranslationFilter=(bX=True,bY=True,bZ=True),RotationFilter=(bX=false,bY=false,bZ=false),ScaleFilter=(bX=false,bY=false,bZ=false))
    Parents=((Item=(Type=Bone,Name="dyn_vest_collar_frnt_l_base"),Weight=1))
    Weight=1
   )
   n7 = &RigUnit_SphericalPoseReader(
    ActiveRegionScaleFactors=(PositiveWidth=1,NegativeWidth=1,PositiveHeight=1,NegativeHeight=1)
    ActiveRegionSize=0.05
    Debug=(bDrawDebug=true,bDraw2D=False,bDrawLocalAxes=False,DebugScale=25,DebugSegments=20,DebugThickness=0.25)
    DriverAxis=(X=-1,Y=0,Z=0)
    DriverItem=(Type=Bone,Name="thigh_l")
    FalloffRegionScaleFactors=(PositiveWidth=1,NegativeWidth=1,PositiveHeight=1,NegativeHeight=1)
    FalloffSize=0.4
    FlipHeightScaling=False
    FlipWidthScaling=False
    OptionalParentItem=(Type=Bone,Name="None")
    RotationOffset=(X=120,Y=10,Z=-5)
    ExecutePin=@RigVMFunction_Sequence.D
   )
   RigVMFunction_MathQuaternionFromEuler = &RigVMFunction_MathQuaternionFromEuler(Euler=(X=0,Y=-15,Z=0))
   > &RigUnit_ModifyTransforms(
    ItemToModify=((Item=(Type=Bone,Name="dyn_vest_side_l_base"),Transform=(Rotation=(X=0,Y=0,Z=0,W=1),Translation=(X=2,Y=-1,Z=3),Scale3D=(X=1,Y=1,Z=1))))
    WeightMinimum=0
    ItemToModify.0.Transform.Rotation=@RigVMFunction_MathQuaternionFromEuler.Result
    Weight=@n7.OutputParam
   )
   n8 = &RigUnit_GetRelativeTransformForItem(Child=(Type=Bone,Name="head"), Parent=(Type=Bone,Name="spine_05"), bChildInitial=False, bParentInitial=False)
   > &RigUnit_ModifyTransforms(
    ItemToModify=((Item=(Type=Bone,Name="dyn_vest_collar_mid_base"),Transform=(Rotation=(X=-0.0655,Y=-0.035,Z=-0.06,W=0.9954),Translation=(X=-0.5612,Y=-1.1116,Z=1.5926),Scale3D=(X=1,Y=1,Z=1))))
    WeightMinimum=0
    Weight=&RigVMFunction_MathDoubleRemap(
     SourceMaximum=-75
     SourceMinimum=-60
     TargetMaximum=1
     TargetMinimum=0
     Value=&RigVMFunction_MathQuaternionToEuler(RotationOrder=ZYX, Value=@n8.RelativeTransform.Rotation).Result.Z
    ).Result
   )
   n9 = &RigUnit_GetRelativeTransformForItem(Child=(Type=Bone,Name="calf_r"), Parent=(Type=Bone,Name="thigh_r"), bChildInitial=False, bParentInitial=False)
   > &RigUnit_ModifyTransforms(
    ItemToModify=((Item=(Type=Bone,Name="kneescaling_r"),Transform=(Rotation=(X=0,Y=0,Z=0,W=1),Translation=(X=0,Y=0,Z=0),Scale3D=(X=-1.5,Y=1,Z=1))))
    WeightMinimum=0
    Weight=&RigVMFunction_MathDoubleRemap(
     SourceMaximum=130
     SourceMinimum=80
     TargetMaximum=1
     TargetMinimum=0
     Value=&RigVMFunction_MathQuaternionToEuler(RotationOrder=ZYX, Value=@n9.RelativeTransform.Rotation).Result.Z
    ).Result
   )
   n10 = &RigVMFunction_Sequence(ExecuteContext=@RigVMFunction_Sequence.E)
   n11 = &RigUnit_GetRelativeTransformForItem(Child=(Type=Bone,Name="foot_l"), Parent=(Type=Bone,Name="calf_l"), bChildInitial=False, bParentInitial=False)
   n12 = &RigUnit_ModifyTransforms(
    ItemToModify=((Item=(Type=Bone,Name="dyn_foot_lace_l_base"),Transform=(Rotation=(X=-0.0899,Y=-0.262,Z=0.521,W=0.8073),Translation=(X=0.7937,Y=0.726,Z=1.2378),Scale3D=(X=1,Y=1,Z=1))))
    WeightMinimum=0
    Weight=&RigVMFunction_MathDoubleRemap(
     SourceMaximum=-30
     SourceMinimum=-20
     TargetMaximum=1
     TargetMinimum=0
     Value=&RigVMFunction_MathQuaternionToEuler(RotationOrder=ZYX, Value=@n11.RelativeTransform.Rotation).Result.Z
    ).Result
    ExecutePin=@n10.A
   )
   n13 = &RigVMFunction_ControlFlowBranch(ExecuteContext=@n12.ExecutePin, Condition=RigVMVariableNode() {variable=Is All Sliding}.Value)
   &RigUnit_ModifyTransforms(
    ItemToModify=((Item=(Type=Bone,Name="dyn_vest_side_l_base"),Transform=(Rotation=(X=-0.3083,Y=0.0934,Z=-0.0829,W=0.9431),Translation=(X=3,Y=0,Z=0),Scale3D=(X=1,Y=1,Z=1))))
    Weight=1
    WeightMinimum=0
    ExecutePin=@n13.True
   )
   > &RigUnit_ModifyTransforms(
    ItemToModify=((Item=(Type=Bone,Name="base_dyn_vest_strap_frnt_l"),Transform=(Rotation=(X=-0.3143,Y=-0.3812,Z=0.3951,W=0.7744),Translation=(X=-5,Y=0,Z=0),Scale3D=(X=1,Y=1,Z=1))))
    Weight=1
    WeightMinimum=0
   )
   n14 = &RigUnit_GetRelativeTransformForItem(Child=(Type=Bone,Name="foot_r"), Parent=(Type=Bone,Name="calf_r"), bChildInitial=False, bParentInitial=False)
   &RigUnit_ModifyTransforms(
    ItemToModify=((Item=(Type=Bone,Name="dyn_foot_lace_r_base"),Transform=(Rotation=(X=0.06,Y=-0.1312,Z=0.3825,W=0.9126),Translation=(X=-0.3867,Y=-1.2354,Z=-1.6508),Scale3D=(X=1,Y=1,Z=1))))
    WeightMinimum=0
    Weight=&RigVMFunction_MathDoubleRemap(
     SourceMaximum=-30
     SourceMinimum=-20
     TargetMaximum=1
     TargetMinimum=0
     Value=&RigVMFunction_MathQuaternionToEuler(RotationOrder=ZYX, Value=@n14.RelativeTransform.Rotation).Result.Z
    ).Result
    ExecutePin=@n10.B
   )
   RigVMFunction_MathBoolNot = &RigVMFunction_MathBoolNot(Value=RigVMVariableNode() {variable=RigidBody Enabled}.Value)
   RigVMFunction_MathBoolAnd = &RigVMFunction_MathBoolAnd(A=@RigVMFunction_MathBoolNot.Result, B=RigVMVariableNode() {variable=Is Tactical Sprinting}.Value)
   Cast = &RigVMFunction_MathBoolToFloat(Value=@RigVMFunction_MathBoolAnd.Result)
   AlphaInterp = &RigVMFunction_AlphaInterp(InterpSpeedDecreasing=10, InterpSpeedIncreasing=10, Scale=1, bClampResult=False, Value=@Cast.Result)
   &RigUnit_ModifyTransforms(
    ItemToModify=((Item=(Type=Bone,Name="AltSkinCorrective"),Transform=(Rotation=(X=0,Y=0,Z=0,W=1),Translation=(X=-5.2152,Y=-15,Z=0),Scale3D=(X=1,Y=1,Z=1))))
    WeightMinimum=0
    Weight=@AlphaInterp.Result
    ExecutePin=@RigVMFunction_Sequence_1.C
   )
   &RigVMFunction_MathQuaternionFromEuler(Euler=(X=0,Y=-55,Z=30))
   &RigVMFunction_MathQuaternionFromEuler(Euler=(X=0,Y=0,Z=0))
   &RigVMFunction_MathQuaternionFromEuler(Euler=(X=0,Y=-90,Z=0))
  ]}
]
hierarchy [
 root {loc=<0 0 0>, rot=<0 0 0 1>}
 pelvis {parent=root, loc=<0 -0.5301 98.0571>, rot=<0.0127 0.707 -0.0127 -0.707>}
 spine_01 {parent=pelvis, loc=<6.1208 -0.9328 0>, rot=<0 0 0.034 -0.9994>}
 spine_02 {parent=spine_01, loc=<6.5996 -0.2095 0>, rot=<0 0 0.0238 -0.9997>}
 spine_03 {parent=spine_02, loc=<6.4914 -0.1237 0>, rot=<0 0 0.0149 -0.9999>}
 spine_04 {parent=spine_03, loc=<6.9598 0.0794 0>, rot=<0 0 -0.0132 -0.9999>}
 spine_05 {parent=spine_04, loc=<6.9987 0 0>, rot=<0 0 -0.0015 -1>}
 attach_cape {parent=spine_05, loc=<9.3216 -8.4501 0>, rot=<0.0616 -0.7044 -0.0616 -0.7044>}
 neck_01 {parent=spine_05, loc=<11.3138 0.1529 0>, rot=<0 0 -0.2003 -0.9797>}
 neck_02 {parent=neck_01, loc=<4.5627 0 0>, rot=<0 0 0.0504 -0.9987>}
 head {parent=neck_02, loc=<5.3914 0 0>, rot=<0 0 0.115 -0.9934>}
 attach_backpack {parent=spine_05, loc=<0 -8.35 0>, rot=<-0.0093 -0.707 0.0093 -0.707>}
 clavicle_l {parent=spine_05, loc=<5.4831 -0.1053 -3.2946>, rot=<0.0465 -0.7289 -0.009 -0.6829>}
 upperarm_l {parent=clavicle_l, loc=<9.8658 -1.2136 -0.4105>, rot=<0.0164 -0.3245 -0.0269 -0.9454>}
 lowerarm_l {parent=upperarm_l, loc=<26.465 -0.8345 0.259>, rot=<0.0036 0.0161 -0.2298 -0.9731>}
 hand_l {parent=lowerarm_l, loc=<22.7355 -0.3203 -0.6105>, rot=<-0.6784 -0.0223 0.0735 -0.7307>}
 thumb_01_l {parent=hand_l, loc=<1.3118 -0.3191 -2.459>, rot=<0.6546 -0.4042 0.0664 -0.6354>}
 thumb_02_l {parent=thumb_01_l, loc=<5.9336 -0.0448 0.0401>, rot=<-0.0037 0.0213 0.298 -0.9543>}
 thumb_03_l {parent=thumb_02_l, loc=<2.2223 0.3682 -0.0907>, rot=<0.0118 -0.013 0.0066 -0.9998>}
 index_metacarpal_l {parent=hand_l, loc=<1.6535 0.0183 -2.96>, rot=<0.0158 -0.0124 -0.0173 -0.9996>}
 index_01_l {parent=index_metacarpal_l, loc=<8.3008 0.0357 -0.0524>, rot=<-0.0044 0.0485 0.0924 -0.9945>}
 index_02_l {parent=index_01_l, loc=<3.3587 -0.3004 0.0452>, rot=<-0.002 0.0027 0.3225 -0.9466>}
 index_03_l {parent=index_02_l, loc=<2.5578 -0.123 -0.0252>, rot=<-0.002 0.0067 0.2558 -0.9667>}
 middle_metacarpal_l {parent=hand_l, loc=<2.1856 -0.1725 -0.5984>, rot=<0.006 0.0168 -0.0329 -0.9993>}
 middle_01_l {parent=middle_metacarpal_l, loc=<7.3275 -0.2447 0.0129>, rot=<-0.0208 0.0042 0.0858 -0.9961>}
 middle_02_l {parent=middle_01_l, loc=<4.0123 0.1401 0.1175>, rot=<-0.0218 0.0258 0.3512 -0.9357>}
 middle_03_l {parent=middle_02_l, loc=<2.6077 0.0534 0.1373>, rot=<0.0082 0.0188 0.1708 -0.9851>}
 ring_metacarpal_l {parent=hand_l, loc=<2.2074 -0.3625 1.3638>, rot=<-0.0284 0.0722 -0.0458 -0.9959>}
 ring_01_l {parent=ring_metacarpal_l, loc=<7.0295 -0.6197 -0.1446>, rot=<-0.0489 -0.0252 0.1892 -0.9804>}
 ring_02_l {parent=ring_01_l, loc=<3.6174 0.4647 -0.1027>, rot=<0 0 0.2806 -0.9598>}
 ring_03_l {parent=ring_02_l, loc=<2.4152 0.114 0.0199>, rot=<0 0 0.1189 -0.9929>}
 pinky_metacarpal_l {parent=hand_l, loc=<1.7616 -0.9753 3.3814>, rot=<-0.0397 0.0814 -0.069 -0.9935>}
 pinky_01_l {parent=pinky_metacarpal_l, loc=<7.1709 -0.4831 -0.1094>, rot=<-0.0654 -0.0035 0.2242 -0.9723>}
 pinky_02_l {parent=pinky_01_l, loc=<2.8274 0.0893 -0.1043>, rot=<0.0037 0.0011 0.2854 -0.9584>}
 pinky_03_l {parent=pinky_02_l, loc=<2.2046 0.0307 -0.0078>, rot=<-0.0002 0.0098 0.1445 -0.9895>}
 deform_wrist_dn_base_l {parent=hand_l, loc=<0 -1.5326 0>, rot=<0 0 0 1>}
 deform_wrist_dn_l {parent=deform_wrist_dn_base_l, loc=<0 0 0>, rot=<0 0 0 1>}
 deform_wrist_up_base_l {parent=hand_l, loc=<0 2.3009 0>, rot=<0 0 0 1>}
 deform_wrist_up_l {parent=deform_wrist_up_base_l, loc=<0 0 0>, rot=<0 0 0 1>}
 dyn_hand_braselet_l {parent=hand_l, loc=<-1.6168 0 0>, rot=<0 0 0 1>}
 lowerarm_twist_02_l {parent=lowerarm_l, loc=<10.6707 0 0>, rot=<0 0 0 1>}
 lowerarm_twist_01_l {parent=lowerarm_l, loc=<18.316 0 0>, rot=<0 0 0 1>}
 deform_elbow_in_base_l {parent=lowerarm_l, loc=<0.1834 3.6062 -0.2572>, rot=<0 0 0 1>}
 deform_elbow_in_l {parent=deform_elbow_in_base_l, loc=<0 0 0>, rot=<0 0 0 1>}
 deform_elbow_out_base_l {parent=lowerarm_l, loc=<0 -4 0>, rot=<0 0 0 1>}
 deform_elbow_out_l {parent=deform_elbow_out_base_l, loc=<0 0 0>, rot=<0 0 0 1>}
 upperarm_twist_01_l {parent=upperarm_l, loc=<7.2135 -0.1134 0>, rot=<0 0 0 1>}
 deform_upperarm_bk_base_l {parent=upperarm_twist_01_l, loc=<-6.4053 -4.3746 0.314>, rot=<0 0 0 1>}
 deform_upperarm_bk_l {parent=deform_upperarm_bk_base_l, loc=<0 0 0>, rot=<0 0 0 1>}
 deform_upperarm_fr_base_l {parent=upperarm_twist_01_l, loc=<-5.6312 5.092 0.3354>, rot=<0 0 0 1>}
 deform_upperarm_fr_l {parent=deform_upperarm_fr_base_l, loc=<0 0 0>, rot=<0 0 0 1>}
 deform_upperarm_up_base_l {parent=upperarm_twist_01_l, loc=<-8.891 0.4825 3.9517>, rot=<0 0 0 1>}
 deform_upperarm_up_l {parent=deform_upperarm_up_base_l, loc=<0 0 0>, rot=<0 0 0 1>}
 upperarm_twist_02_l {parent=upperarm_l, loc=<19.7998 -0.1134 0>, rot=<0 0 0 1>}
 dyn_upperarm_vest_strap_l_01 {parent=upperarm_l, loc=<7.1955 -2.9545 3.0078>, rot=<-0.3069 0.0257 0.0666 -0.9491>}
 dyn_upperarm_vest_strap_l_02 {parent=dyn_upperarm_vest_strap_l_01, loc=<4.964 0 0>, rot=<0 0 0.0225 -0.9997>}
 dyn_upperarm_ALT_correct_l {parent=upperarm_l, loc=<5.2119 0 0>, rot=<0 0 0 1>}
 dyn_vest_collar_frnt_l_base {parent=clavicle_l, loc=<6.0339 9.0946 -3.6738>, rot=<0.5618 0.0424 -0.0027 -0.8262>}
 dyn_vest_collar_frnt_l_01 {parent=dyn_vest_collar_frnt_l_base, loc=<0 0 0>, rot=<0 0 0 1>}
 dyn_vest_collar_l_base {parent=clavicle_l, loc=<5.8577 -0.3818 4.978>, rot=<0.2499 0.1481 0.0037 -0.9569>}
 dyn_vest_collar_l_01 {parent=dyn_vest_collar_l_base, loc=<0 0 0>, rot=<0 0 0 1>}
 dyn_vest_collar_strap_l_base {parent=dyn_vest_collar_l_01, loc=<9.2384 0.8244 -1.2654>, rot=<0.0226 -0.1567 -0.4204 -0.8934>}
 dyn_vest_collar_strap_l_01 {parent=dyn_vest_collar_strap_l_base, loc=<0 0 0>, rot=<0 0 0 1>}
 dyn_vest_collar_frnt_l_support {parent=clavicle_l, loc=<6.0339 9.0946 -3.6738>, rot=<0.5618 0.0424 -0.0027 -0.8262>}
 clavicle_r {parent=spine_05, loc=<5.4831 -0.1053 3.2946>, rot=<0.7289 0.0465 -0.6829 0.009>}
 upperarm_r {parent=clavicle_r, loc=<-9.8658 1.2136 0.4105>, rot=<0.0164 -0.3245 -0.0269 -0.9454>}
 lowerarm_r {parent=upperarm_r, loc=<-26.465 0.8345 -0.259>, rot=<0.0036 0.0161 -0.2298 -0.9731>}
 hand_r {parent=lowerarm_r, loc=<-22.7355 0.3203 0.6106>, rot=<-0.6784 -0.0223 0.0735 -0.7307>}
 thumb_01_r {parent=hand_r, loc=<-1.3118 0.3191 2.459>, rot=<0.6546 -0.4042 0.0664 -0.6354>}
 thumb_02_r {parent=thumb_01_r, loc=<-5.9336 0.0448 -0.0401>, rot=<-0.0037 0.0213 0.298 -0.9543>}
 thumb_03_r {parent=thumb_02_r, loc=<-2.2223 -0.3682 0.0907>, rot=<0.0118 -0.013 0.0066 -0.9998>}
 index_metacarpal_r {parent=hand_r, loc=<-1.6535 -0.0183 2.96>, rot=<0.0158 -0.0124 -0.0173 -0.9996>}
 index_01_r {parent=index_metacarpal_r, loc=<-8.3008 -0.0357 0.0524>, rot=<-0.0044 0.0485 0.0924 -0.9945>}
 index_02_r {parent=index_01_r, loc=<-3.3587 0.3004 -0.0452>, rot=<-0.002 0.0027 0.3225 -0.9466>}
 index_03_r {parent=index_02_r, loc=<-2.5578 0.123 0.0252>, rot=<-0.002 0.0067 0.2558 -0.9667>}
 middle_metacarpal_r {parent=hand_r, loc=<-2.1856 0.1725 0.5984>, rot=<0.006 0.0168 -0.0329 -0.9993>}
 middle_01_r {parent=middle_metacarpal_r, loc=<-7.3275 0.2447 -0.0129>, rot=<-0.0208 0.0042 0.0858 -0.9961>}
 middle_02_r {parent=middle_01_r, loc=<-4.0123 -0.1401 -0.1175>, rot=<-0.0218 0.0258 0.3512 -0.9357>}
 middle_03_r {parent=middle_02_r, loc=<-2.6077 -0.0534 -0.1373>, rot=<0.0082 0.0188 0.1708 -0.9851>}
 ring_metacarpal_r {parent=hand_r, loc=<-2.2074 0.3625 -1.3638>, rot=<-0.0284 0.0722 -0.0458 -0.9959>}
 ring_01_r {parent=ring_metacarpal_r, loc=<-7.0295 0.6197 0.1446>, rot=<-0.0489 -0.0252 0.1892 -0.9804>}
 ring_02_r {parent=ring_01_r, loc=<-3.6174 -0.4647 0.1027>, rot=<0 0 0.2806 -0.9598>}
 ring_03_r {parent=ring_02_r, loc=<-2.4152 -0.114 -0.0199>, rot=<0 0 0.1189 -0.9929>}
 pinky_metacarpal_r {parent=hand_r, loc=<-1.7616 0.9753 -3.3814>, rot=<-0.0397 0.0814 -0.069 -0.9935>}
 pinky_01_r {parent=pinky_metacarpal_r, loc=<-7.1709 0.4831 0.1094>, rot=<-0.0654 -0.0035 0.2242 -0.9723>}
 pinky_02_r {parent=pinky_01_r, loc=<-2.8274 -0.0893 0.1043>, rot=<0.0037 0.0011 0.2854 -0.9584>}
 pinky_03_r {parent=pinky_02_r, loc=<-2.2046 -0.0307 0.0078>, rot=<-0.0002 0.0098 0.1445 -0.9895>}
 deform_wrist_dn_base_r {parent=hand_r, loc=<-0.0001 1.5328 0>, rot=<0 0 0 1>}
 deform_wrist_dn_r {parent=deform_wrist_dn_base_r, loc=<0 0 0>, rot=<0 0 0 1>}
 deform_wrist_up_base_r {parent=hand_r, loc=<-0.0002 -2.3007 -0.0001>, rot=<0 0 0 1>}
 deform_wrist_up_r {parent=deform_wrist_up_base_r, loc=<0 0 0>, rot=<0 0 0 1>}
 lowerarm_twist_02_r {parent=lowerarm_r, loc=<-10.6707 0 0>, rot=<0 0 0 -1>}
 lowerarm_braselet_l_01 {parent=lowerarm_twist_02_r, loc=<0.067 -0.6385 -3.9949>, rot=<0.087 -0.0805 0.007 -0.9929>}
 lowerarm_twist_01_r {parent=lowerarm_r, loc=<-18.316 0 0>, rot=<0 0 0 -1>}
 deform_elbow_in_base_r {parent=lowerarm_r, loc=<-0.1834 -3.6062 0.2572>, rot=<0 0 0 -1>}
 deform_elbow_in_r {parent=deform_elbow_in_base_r, loc=<0 0 0>, rot=<0 0 0 1>}
 deform_elbow_out_base_r {parent=lowerarm_r, loc=<0 4 -0.0001>, rot=<0 0 0 -1>}
 deform_elbow_out_r {parent=deform_elbow_out_base_r, loc=<0 0 0>, rot=<0 0 0 1>}
 upperarm_twist_01_r {parent=upperarm_r, loc=<-7.2135 0.1134 0>, rot=<0 0 0 -1>}
 deform_upperarm_bk_base_r {parent=upperarm_twist_01_r, loc=<6.4053 4.3746 -0.314>, rot=<0 0 0 1>}
 deform_upperarm_bk_r {parent=deform_upperarm_bk_base_r, loc=<0 0 0>, rot=<0 0 0 1>}
 deform_upperarm_fr_base_r {parent=upperarm_twist_01_r, loc=<5.6312 -5.092 -0.3354>, rot=<0 0 0 1>}
 deform_upperarm_fr_r {parent=deform_upperarm_fr_base_r, loc=<0 0 0>, rot=<0 0 0 1>}
 deform_upperarm_up_base_r {parent=upperarm_twist_01_r, loc=<8.891 -0.4825 -3.9517>, rot=<0 0 0 1>}
 deform_upperarm_up_r {parent=deform_upperarm_up_base_r, loc=<0 0 0>, rot=<0 0 0 1>}
 upperarm_twist_02_r {parent=upperarm_r, loc=<-19.7998 0.1134 0>, rot=<0 0 0 -1>}
 dyn_upperarm_vest_strap_r_01 {parent=upperarm_r, loc=<-7.1953 2.9545 -3.0081>, rot=<-0.3069 0.0257 0.0666 -0.9491>}
 dyn_upperarm_vest_strap_r_02 {parent=dyn_upperarm_vest_strap_r_01, loc=<-4.9644 0.0003 0.0005>, rot=<0 0 0.0225 -0.9997>}
 dyn_upperarm_ALT_correct_r {parent=upperarm_r, loc=<-5.2118 0 -0.0001>, rot=<0 0 0 -1>}
 dyn_vest_collar_frnt_r_base {parent=clavicle_r, loc=<-6.0339 -9.0946 3.6735>, rot=<0.5618 0.0424 -0.0027 -0.8262>}
 dyn_vest_collar_frnt_r_01 {parent=dyn_vest_collar_frnt_r_base, loc=<0 0 0>, rot=<0 0 0 1>}
 dyn_vest_collar_r_base {parent=clavicle_r, loc=<-5.8577 0.3818 -4.9782>, rot=<0.2499 0.1481 0.0037 -0.9569>}
 dyn_vest_collar_r_01 {parent=dyn_vest_collar_r_base, loc=<0 0 0>, rot=<0 0 0 1>}
 dyn_vest_collar_strap_r_base {parent=dyn_vest_collar_r_01, loc=<-9.2383 -0.6864 1.2659>, rot=<0.0286 -0.1557 -0.3856 -0.909>}
 dyn_vest_collar_strap_r_01 {parent=dyn_vest_collar_strap_r_base, loc=<0 0 0>, rot=<0 0 0 1>}
 dyn_vest_collar_frnt_r_support {parent=clavicle_r, loc=<-6.0339 -9.0946 3.6735>, rot=<0.5618 0.0424 -0.0027 -0.8262>}
 dyn_vest_collar_mid_base {parent=spine_05, loc=<8.8945 -10.9789 0>, rot=<0 0 0.1491 -0.9888>}
 dyn_vest_collar_mid_01 {parent=dyn_vest_collar_mid_base, loc=<0 0 0>, rot=<0 0 0 1>}
 dyn_vest_collar_mid_02 {parent=dyn_vest_collar_mid_01, loc=<4.7031 0 0>, rot=<0 0 -0.1062 -0.9943>}
 dyn_vest_strap_frnt_r_01 {parent=spine_05, loc=<2.1155 6.4177 12.825>, rot=<0.8902 -0.0483 -0.1 0.4418>}
 dyn_vest_strap_frnt_r_02 {parent=dyn_vest_strap_frnt_r_01, loc=<-5.8149 0 0.0002>, rot=<0 0 0 1>}
 dyn_top_correct_LOD1 {parent=spine_05, loc=<5.5518 6.0381 0>, rot=<-0.6419 -0.2966 -0.2966 -0.6419>}
 base_dyn_vest_strap_frnt_l {parent=spine_05, loc=<2.1151 6.4177 -12.825>, rot=<-0.0483 -0.8902 -0.4418 -0.1>}
 dyn_vest_strap_frnt_l_01 {parent=base_dyn_vest_strap_frnt_l, loc=<0 0 0>, rot=<0 0 0 1>}
 dyn_vest_strap_frnt_l_02 {parent=dyn_vest_strap_frnt_l_01, loc=<5.8143 0 0>, rot=<0 0 0 1>}
 deform_pec_aim_base_l {parent=spine_04, loc=<3.2456 9.3309 -1.2237>, rot=<0.0603 -0.4033 0.3318 -0.8507>}
 deform_pec_aim_l {parent=deform_pec_aim_base_l, loc=<0 0 0>, rot=<0 0 0 1>}
 deform_pec_base_l {parent=deform_pec_aim_l, loc=<9.3665 0 0>, rot=<0 0 0 1>}
 deform_pec_l {parent=deform_pec_base_l, loc=<0 0 0>, rot=<0 0 0 1>}
 deform_pec_aim_base_r {parent=spine_04, loc=<3.2452 9.3309 1.2237>, rot=<-0.4033 -0.0603 0.8507 0.3318>}
 deform_pec_aim_r {parent=deform_pec_aim_base_r, loc=<0 0 0>, rot=<0 0 0 1>}
 deform_pec_base_r {parent=deform_pec_aim_r, loc=<-9.3667 -0.0003 -0.0004>, rot=<0 0 0 1>}
 deform_pec_r {parent=deform_pec_base_r, loc=<0 0 0>, rot=<0 0 0 1>}
 dyn_vest_strap_support_r {parent=spine_04, loc=<-0.8829 -0.7662 13.2242>, rot=<0.7065 0.0294 -0.7065 0.0294>}
 dyn_vest_strap_support_l {parent=spine_04, loc=<-0.8829 -0.7662 -13.2242>, rot=<0.0294 -0.7065 -0.0294 -0.7065>}
 dyn_vest_bk_l_base {parent=spine_03, loc=<0.9603 -6.2478 -8.5431>, rot=<0.2929 -0.6725 0.2506 -0.6318>}
 dyn_vest_bk_l_01 {parent=dyn_vest_bk_l_base, loc=<0 0 0>, rot=<0 -0.7071 0 -0.7071>}
 dyn_vest_bk_l_02 {parent=dyn_vest_bk_l_01, loc=<6.8725 0 1.3303>, rot=<0 0.144 0 -0.9896>}
 dyn_vest_bk_l_03 {parent=dyn_vest_bk_l_02, loc=<13.6523 0 0>, rot=<0 0 0 1>}
 dyn_vest_bk_l_04 {parent=dyn_vest_bk_l_03, loc=<16.7031 0 0>, rot=<0 0 0 1>}
 dyn_strap_bk_l_01 {parent=dyn_vest_bk_l_04, loc=<13.4077 -7.1719 -2.2981>, rot=<0.5317 -0.064 -0.0215 -0.8443>}
 dyn_vest_side_r_base {parent=spine_03, loc=<1.0511 1.8717 12.7488>, rot=<0 0 0.9985 -0.0547>}
 dyn_vest_side_r_01 {parent=dyn_vest_side_r_base, loc=<0 0 0>, rot=<0 0 0 1>}
 dyn_vest_side_r_02 {parent=dyn_vest_side_r_01, loc=<7.3708 0 -0.2296>, rot=<0 0.0864 0 -0.9963>}
 dyn_vest_side_r_03 {parent=dyn_vest_side_r_02, loc=<14.1375 0 0>, rot=<0 0.0407 0 -0.9992>}
 dyn_vest_side_r_04 {parent=dyn_vest_side_r_03, loc=<17.1994 0 0.0001>, rot=<0 0 0 1>}
 dyn_strap_side_r_02 {parent=dyn_vest_side_r_04, loc=<13.8411 -2.8544 0.081>, rot=<-0.0598 -0.029 0.0274 -0.9974>}
 dyn_strap_side_r_03 {parent=dyn_strap_side_r_02, loc=<4.3084 -0.0811 0>, rot=<0 0 0 1>}
 dyn_strap_side_r_01 {parent=dyn_vest_side_r_04, loc=<13.6781 3.5501 -1.0225>, rot=<0.1684 -0.021 -0.0251 -0.9852>}
 dyn_vest_side_l_base {parent=spine_03, loc=<1.0809 1.875 -12.7488>, rot=<0.0387 -0.706 -0.0387 -0.706>}
 dyn_vest_side_l_01 {parent=dyn_vest_side_l_base, loc=<0 0 0>, rot=<0 -0.7071 0 -0.7071>}
 dyn_vest_side_l_02 {parent=dyn_vest_side_l_01, loc=<7.0565 0 0.301>, rot=<0 0.0952 0 -0.9955>}
 dyn_vest_side_l_03 {parent=dyn_vest_side_l_02, loc=<14.8184 0 0>, rot=<0 0 0 1>}
 dyn_vest_side_l_04 {parent=dyn_vest_side_l_03, loc=<16.699 0 0.4292>, rot=<0 0 0 1>}
 dyn_strap_side_l_01 {parent=dyn_vest_side_l_04, loc=<13.3256 -3.9789 -1.3421>, rot=<-0.1295 -0.0335 0.0214 -0.9908>}
 dyn_strap_side_l_02 {parent=dyn_strap_side_l_01, loc=<4.3006 -0.1704 0.2292>, rot=<0 0 0.0892 -0.996>}
 dyn_strap_bk_l_06 {parent=dyn_vest_side_l_02, loc=<5.4569 -6.3636 -0.1007>, rot=<-0.3108 0.1947 0.1238 -0.9221>}
 dyn_quiver_l_base {parent=dyn_strap_bk_l_06, loc=<8.7135 2.4553 -1.3311>, rot=<0.055 -0.1407 0.4282 -0.891>}
 dyn_arrow_l_02 {parent=dyn_quiver_l_base, loc=<-7.6792 -0.4488 -0.7427>, rot=<0 0 0.0405 -0.9992>}
 dyn_vest_bk_mid_base {parent=spine_03, loc=<0.9934 -8.8802 0.0301>, rot=<0.5484 -0.4457 0.4479 -0.5476>}
 dyn_vest_bk_mid_01 {parent=dyn_vest_bk_mid_base, loc=<0 0 0.0086>, rot=<0 -0.7071 0 -0.7071>}
 dyn_vest_bk_mid_02 {parent=dyn_vest_bk_mid_01, loc=<7.4265 0 -0.0398>, rot=<0 0.0304 0 -0.9995>}
 dyn_vest_bk_mid_03 {parent=dyn_vest_bk_mid_02, loc=<13.252 0 0.0037>, rot=<0 0.0488 0 -0.9988>}
 dyn_vest_bk_mid_04 {parent=dyn_vest_bk_mid_03, loc=<16.727 0 0>, rot=<0 0 0 1>}
 dyn_vest_bk_r_base {parent=spine_03, loc=<0.9945 -6.1709 8.4705>, rot=<0.6318 -0.2506 0.6725 -0.2929>}
 dyn_vest_bk_r_01 {parent=dyn_vest_bk_r_base, loc=<0 0 0>, rot=<0 -0.7071 0 -0.7071>}
 dyn_vest_bk_r_02 {parent=dyn_vest_bk_r_01, loc=<7.161 0 0.6322>, rot=<0 0.144 0 -0.9896>}
 dyn_vest_bk_r_03 {parent=dyn_vest_bk_r_02, loc=<13.5848 0 0>, rot=<0 0 0 1>}
 dyn_vest_bk_r_04 {parent=dyn_vest_bk_r_03, loc=<16.718 0 0.598>, rot=<0 0 0 1>}
 dyn_strap_bk_r_01 {parent=dyn_vest_bk_r_04, loc=<13.4461 5.3683 -1.0662>, rot=<0.7956 -0.0029 -0.0278 -0.6051>}
 dyn_strap_bk_r_02 {parent=dyn_strap_bk_r_01, loc=<3.9839 0 0>, rot=<0.0318 0.0152 -0.0005 -0.9994>}
 thigh_l {parent=pelvis, loc=<-1.0855 -0.2109 -8.7053>, rot=<0.0009 0.0355 0.0155 -0.9993>}
 calf_l {parent=thigh_l, loc=<-42.5284 -0.0019 -0.7297>, rot=<0.0006 -0.0161 -0.0391 -0.9991>}
 foot_l {parent=calf_l, loc=<-46.7401 -2.5014 -0.592>, rot=<0.0427 -0.0238 0.0353 -0.9982>}
 ball_l {parent=foot_l, loc=<-3.5658 10.8066 -0.1544>, rot=<0.0001 -0.0001 -0.7063 -0.7079>}
 dyn_foot_lace_l_base {parent=foot_l, loc=<4.0147 6.689 0.2892>, rot=<0.0837 0.0247 -0.0006 -0.9962>}
 dyn_foot_lace_l_01 {parent=dyn_foot_lace_l_base, loc=<0 0 -1.0991>, rot=<0 -0.6899 0 -0.7239>}
 dyn_foot_lace_l_02 {parent=dyn_foot_lace_l_base, loc=<0 0 1.0757>, rot=<0 0.6922 0 -0.7217>}
 calf_twist_02_l {parent=calf_l, loc=<-19.1546 -1.7943 -0.2638>, rot=<0.0237 0.0095 -0.0035 -0.9997>}
 calf_twist_01_l {parent=calf_l, loc=<-35.8483 -1.9197 -0.5794>, rot=<0.0237 0.0095 -0.0035 -0.9997>}
 deform_knee_in_base_l {parent=calf_l, loc=<0.8031 -5.0231 0.2456>, rot=<0 0 0 -1>}
 deform_knee_in_l {parent=deform_knee_in_base_l, loc=<0 0 0>, rot=<0 0 0 1>}
 deform_knee_out_base_l {parent=calf_l, loc=<0 4 0>, rot=<0 0 0 -1>}
 deform_knee_out_l {parent=deform_knee_out_base_l, loc=<0 0 0>, rot=<0 0 0 1>}
 kneescaling_l {parent=calf_l, loc=<0 0 0>, rot=<0 0 0 1>}
 thigh_twist_01_l {parent=thigh_l, loc=<-27.952 0 0>, rot=<0.0013 0 0 -1>}
 deform_groin_base_l {parent=thigh_l, loc=<0.4801 5.6776 -0.2486>, rot=<0 0 0 -1>}
 deform_groin_l {parent=deform_groin_base_l, loc=<0 0 0>, rot=<0 0 0 1>}
 thigh_r {parent=pelvis, loc=<-1.0855 -0.2109 8.7053>, rot=<0.0355 -0.0009 0.9993 0.0155>}
 calf_r {parent=thigh_r, loc=<42.5284 0.0019 0.7297>, rot=<0.0006 -0.0161 -0.0391 -0.9991>}
 foot_r {parent=calf_r, loc=<46.7401 2.5014 0.592>, rot=<0.0427 -0.0238 0.0353 -0.9982>}
 ball_r {parent=foot_r, loc=<3.5658 -10.8066 0.1544>, rot=<0.0001 -0.0001 -0.7063 -0.7079>}
 dyn_foot_lace_r_base {parent=foot_r, loc=<-4.0147 -6.689 -0.2892>, rot=<0.0837 0.0247 -0.0006 -0.9962>}
 dyn_foot_lace_r_01 {parent=dyn_foot_lace_r_base, loc=<0 0 1.0991>, rot=<0 -0.6899 0 -0.7239>}
 dyn_foot_lace_r_02 {parent=dyn_foot_lace_r_base, loc=<0.0001 0 -1.0758>, rot=<0 0.6922 0 -0.7217>}
 calf_twist_02_r {parent=calf_r, loc=<19.1546 1.7943 0.2638>, rot=<0.0237 0.0095 -0.0035 -0.9997>}
 calf_twist_01_r {parent=calf_r, loc=<35.8483 1.9197 0.5794>, rot=<0.0237 0.0095 -0.0035 -0.9997>}
 deform_knee_in_base_r {parent=calf_r, loc=<-0.8031 5.0231 -0.2457>, rot=<0 0 0 -1>}
 deform_knee_in_r {parent=deform_knee_in_base_r, loc=<0 0 0>, rot=<0 0 0 1>}
 deform_knee_out_base_r {parent=calf_r, loc=<0 -4 0>, rot=<0 0 0 -1>}
 deform_knee_out_r {parent=deform_knee_out_base_r, loc=<0 0 0>, rot=<0 0 0 1>}
 kneescaling_r {parent=calf_r, loc=<0 0 0>, rot=<-0.6927 0.0276 0.7201 0.0309>}
 thigh_twist_01_r {parent=thigh_r, loc=<27.952 0 0>, rot=<0.0013 0 0 -1>}
 deform_groin_base_r {parent=thigh_r, loc=<-0.4801 -5.6776 0.2486>, rot=<0 0 0 -1>}
 deform_groin_r {parent=deform_groin_base_r, loc=<0 0 0>, rot=<0 0 0 1>}
 deform_glute_base_l {parent=pelvis, loc=<-7.3466 -9.0213 -9.1245>, rot=<0.0172 0.031 -0.4477 -0.8935>}
 deform_glute_l {parent=deform_glute_base_l, loc=<0 0 0>, rot=<0 0 0 1>}
 deform_glute_base_r {parent=pelvis, loc=<-7.3464 -9.0213 9.1245>, rot=<0.031 -0.0172 0.8935 -0.4477>}
 deform_glute_r {parent=deform_glute_base_r, loc=<0 0 0>, rot=<0 0 0 1>}
 AltSkinCorrective {parent=pelvis, loc=<-4.5729 -14.3153 0>, rot=<0 0 0 1>}
]
variables [
 Is All Sliding {type=bool, default=False}
 Is Tactical Sprinting {type=bool, default=False}
 RigidBody Enabled {type=bool, default=False}
 Weight {type=float, default=1}
]
