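# Physics asset for the female medium SparkArcher player body, set up for physics motion of the hanging vest.
# Five back and side chains each run from a base through four capsule links, with tight swing and twist limits that mostly loosen toward the tip.
# Angularly free constraints join neighboring chains at the lower two rows, tying them into a sheet.
# Pelvis, thigh and right-calf capsules, plus a large offset root sphere, provide body shapes; the left calf has none.
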PhysicsAsset /Game/Characters/Player/Female/Medium/Bodies/F_MED_SparkArcher/Meshes/F_MED_SparkArcher_Physics
bodies {
 calf_r [{shapeType=Capsule, length=45.0377, radius=7.109, shapeName=calf_r_capsule, center=<21.8202 2.8973 0.2862>, rotation=<rot 90 180 175.5783>}]
 dyn_vest_bk_l_01 [{shapeType=Capsule, length=5, radius=2, shapeName=dyn_vest_bk_l_01_capsule, center=<3.6213 0 0.6657>, rotation=<rot 84.5291 180 179.9999>}]
 dyn_vest_bk_l_02 [{
   shapeType=Capsule
   length=9.0765
   radius=7.2617
   shapeName=dyn_vest_bk_l_02_capsule
   center=<7.9247 0.0341 6.5951>
   rotation=<rot 90 180 -92.4748>
  }]
 dyn_vest_bk_l_03 [{
   shapeType=Capsule
   length=13.2009
   radius=8.3916
   shapeName=dyn_vest_bk_l_03_capsule
   center=<8.7012 0 5.8851>
   rotation=<rot 90 -159.444 110.7547>
  }]
 dyn_vest_bk_l_04 [{shapeType=Capsule, length=15, radius=10.2617, shapeName=dyn_vest_bk_l_04_capsule, center=<7.0899 0 5.7296>, rotation=<rot 0 0 90>}]
 dyn_vest_bk_l_base [{shapeType=Capsule, length=0, radius=0, shapeName=dyn_vest_bk_l_base_capsule}]
 dyn_vest_bk_mid_01 [{shapeType=Capsule, length=5, radius=2, shapeName=dyn_vest_bk_mid_01_capsule, center=<3.685 0 0>, rotation=<rot 90 0 0>}]
 dyn_vest_bk_mid_02 [{shapeType=Capsule, length=9, radius=7.2617, shapeName=dyn_vest_bk_mid_02_capsule, center=<7.1362 0 5.152>, rotation=<rot 90 180 -90>}]
 dyn_vest_bk_mid_03 [{
   shapeType=Capsule
   length=13.6637
   radius=8.2617
   shapeName=dyn_vest_bk_mid_03_capsule
   center=<8.7012 0 3.6909>
   rotation=<rot 90 165.9638 76.1625>
  }]
 dyn_vest_bk_mid_04 [{shapeType=Capsule, length=15, radius=10.2617, shapeName=dyn_vest_bk_mid_04_capsule, center=<7.0899 0 0>, rotation=<rot 0 0 90>}]
 dyn_vest_bk_mid_base [{shapeType=Capsule, length=0, radius=0, shapeName=dyn_vest_bk_mid_base_capsule}]
 dyn_vest_bk_r_01 [{shapeType=Capsule, length=5, radius=2, shapeName=dyn_vest_bk_r_01_capsule, center=<3.685 0 0>, rotation=<rot 90 0 0>}]
 dyn_vest_bk_r_02 [{shapeType=Capsule, length=9, radius=7.2617, shapeName=dyn_vest_bk_r_02_capsule, center=<7.1362 0 4.5>, rotation=<rot 90 175.2364 -91.5713>}]
 dyn_vest_bk_r_03 [{
   shapeType=Capsule
   length=13.6637
   radius=8.2617
   shapeName=dyn_vest_bk_r_03_capsule
   center=<8.7012 0 3.6909>
   rotation=<rot 90 165.9638 76.1626>
  }]
 dyn_vest_bk_r_04 [{shapeType=Capsule, length=15, radius=10.2617, shapeName=dyn_vest_bk_r_04_capsule, center=<7.0899 0 4.8957>, rotation=<rot 0 0 90>}]
 dyn_vest_bk_r_base [{shapeType=Capsule, length=0, radius=0, shapeName=dyn_vest_bk_r_base_capsule}]
 dyn_vest_side_l_01 [{shapeType=Capsule, length=5, radius=2, shapeName=dyn_vest_side_l_01_capsule, center=<3.685 0 0>, rotation=<rot 90 0 0>}]
 dyn_vest_side_l_02 [
  {
   shapeType=Capsule
   length=10.0931
   radius=7.2617
   shapeName=dyn_vest_side_l_02_capsule
   center=<7.9518 0 6.3402>
   rotation=<rot 90 -165.9638 -75.9638>
  }
  {shapeType=Box, extentX=19.3673, extentY=25.9852, extentZ=3.1892, shapeName=dyn_vest_side_l_02_box, center=<6.1432 0 0>}
 ]
 dyn_vest_side_l_03 [{
   shapeType=Capsule
   length=14.6507
   radius=8.2617
   shapeName=dyn_vest_side_l_03_capsule
   center=<8.7012 0 6.6974>
   rotation=<rot 90 -165.9638 -75.9637>
  }]
 dyn_vest_side_l_04 [{shapeType=Capsule, length=15, radius=10.2617, shapeName=dyn_vest_side_l_04_capsule, center=<7.0899 0 7.1803>, rotation=<rot 0 0 90>}]
 dyn_vest_side_l_base [{shapeType=Capsule, length=0, radius=0, shapeName=dyn_vest_side_l_base_capsule}]
 dyn_vest_side_r_01 [{shapeType=Capsule, length=5, radius=2, shapeName=dyn_vest_side_r_01_capsule, center=<3.685 0 0>, rotation=<rot 90 0 0>}]
 dyn_vest_side_r_02 [{shapeType=Capsule, length=9, radius=7.2617, shapeName=dyn_vest_side_r_02_capsule, center=<7.1362 0 5.5>, rotation=<rot 90 180 -86.5969>}]
 dyn_vest_side_r_03 [{
   shapeType=Capsule
   length=13.6637
   radius=8.2617
   shapeName=dyn_vest_side_r_03_capsule
   center=<8.7012 0 7.0226>
   rotation=<rot 90 -165.9638 -73.7242>
  }]
 dyn_vest_side_r_04 [{shapeType=Capsule, length=15, radius=10.2617, shapeName=dyn_vest_side_r_04_capsule, center=<7.0899 0 7.1803>, rotation=<rot 0 0 90>}]
 dyn_vest_side_r_base [{shapeType=Capsule, length=0, radius=0, shapeName=dyn_vest_side_r_base_capsule}]
 pelvis [{shapeType=Capsule, length=10.0287, radius=9.3976, shapeName=pelvis_capsule, center=<2.2091 -2.2433 0>}]
 root [{shapeType=Sphere, radius=130.4297, shapeName=root_sphere, center=<0 -17.3743 -123.1262>}]
 thigh_l [{
   shapeType=Capsule
   length=43.235
   radius=7.8594
   shapeName=thigh_l_capsule
   center=<-20.0368 -1 -0.193>
   rotation=<rot 87.3906 179.9989 179.9993>
  }]
 thigh_r [
  {
   shapeType=Capsule
   length=31.6199
   radius=6.8065
   shapeName=thigh_r_capsule
   center=<19.9738 -1 0.35>
   rotation=<rot 87.3906 179.9989 179.9993>
  }
  {shapeType=Capsule, length=6.8392, radius=7.4832, shapeName=thigh_r_capsule, center=<14.599 -3.9623 -1.9431>}
 ]
}
constraints [
 {bone1Name=dyn_vest_bk_l_01, bone2Name=dyn_vest_bk_l_base, swing1LimitDegrees=5, swing2LimitDegrees=5, twistLimitDegrees=0}
 {bone1Name=dyn_vest_bk_l_02, bone2Name=dyn_vest_bk_l_01, swing1LimitDegrees=10, swing2LimitDegrees=5, twistLimitDegrees=2}
 {bone1Name=dyn_vest_bk_l_03, bone2Name=dyn_vest_bk_l_02, swing1LimitDegrees=10, swing2LimitDegrees=10, twistLimitDegrees=2}
 {bone1Name=dyn_vest_bk_l_04, bone2Name=dyn_vest_bk_l_03, swing1LimitDegrees=20, swing2LimitDegrees=20, twistLimitDegrees=2}
 {bone1Name=dyn_vest_bk_mid_01, bone2Name=dyn_vest_bk_mid_base, swing1LimitDegrees=5, swing2LimitDegrees=5, twistLimitDegrees=0}
 {bone1Name=dyn_vest_bk_mid_04, bone2Name=dyn_vest_bk_mid_03, swing1LimitDegrees=10, swing2LimitDegrees=20, twistLimitDegrees=2}
 {bone1Name=dyn_vest_bk_mid_03, bone2Name=dyn_vest_bk_mid_02, swing1LimitDegrees=10, swing2LimitDegrees=10, twistLimitDegrees=2}
 {bone1Name=dyn_vest_bk_mid_02, bone2Name=dyn_vest_bk_mid_01, swing1LimitDegrees=10, swing2LimitDegrees=5, twistLimitDegrees=2}
 {bone1Name=dyn_vest_bk_r_01, bone2Name=dyn_vest_bk_r_base, swing1LimitDegrees=5, swing2LimitDegrees=5, twistLimitDegrees=0}
 {bone1Name=dyn_vest_bk_r_04, bone2Name=dyn_vest_bk_r_03, swing1LimitDegrees=20, swing2LimitDegrees=20, twistLimitDegrees=2}
 {bone1Name=dyn_vest_bk_r_03, bone2Name=dyn_vest_bk_r_02, swing1LimitDegrees=10, swing2LimitDegrees=10, twistLimitDegrees=2}
 {bone1Name=dyn_vest_bk_r_02, bone2Name=dyn_vest_bk_r_01, swing1LimitDegrees=10, swing2LimitDegrees=5, twistLimitDegrees=2}
 {bone1Name=dyn_vest_side_l_01, bone2Name=dyn_vest_side_l_base, swing1LimitDegrees=5, swing2LimitDegrees=5, twistLimitDegrees=0}
 {bone1Name=dyn_vest_side_l_02, bone2Name=dyn_vest_side_l_01, swing1LimitDegrees=10, swing2LimitDegrees=5, twistLimitDegrees=2}
 {bone1Name=dyn_vest_side_l_03, bone2Name=dyn_vest_side_l_02, swing1LimitDegrees=10, swing2LimitDegrees=10, twistLimitDegrees=2}
 {bone1Name=dyn_vest_side_l_04, bone2Name=dyn_vest_side_l_03, swing1LimitDegrees=10, swing2LimitDegrees=10, twistLimitDegrees=2}
 {bone1Name=dyn_vest_side_r_01, bone2Name=dyn_vest_side_r_base, swing1LimitDegrees=5, swing2LimitDegrees=5, twistLimitDegrees=0}
 {bone1Name=dyn_vest_side_r_04, bone2Name=dyn_vest_side_r_03, swing1LimitDegrees=10, swing2LimitDegrees=10, twistLimitDegrees=2}
 {bone1Name=dyn_vest_side_r_03, bone2Name=dyn_vest_side_r_02, swing1LimitDegrees=10, swing2LimitDegrees=10, twistLimitDegrees=2}
 {bone1Name=dyn_vest_side_r_02, bone2Name=dyn_vest_side_r_01, swing1LimitDegrees=10, swing2LimitDegrees=5, twistLimitDegrees=2}
 {
  bone1Name=dyn_vest_bk_l_04
  bone2Name=dyn_vest_side_l_04
  swing1LimitDegrees=45
  swing1Motion=Free
  swing2LimitDegrees=45
  swing2Motion=Free
  twistLimitDegrees=45
  twistMotion=Free
 }
 {
  bone1Name=dyn_vest_bk_mid_04
  bone2Name=dyn_vest_bk_l_04
  swing1LimitDegrees=45
  swing1Motion=Free
  swing2LimitDegrees=45
  swing2Motion=Free
  twistLimitDegrees=45
  twistMotion=Free
 }
 {
  bone1Name=dyn_vest_bk_r_04
  bone2Name=dyn_vest_bk_mid_04
  swing1LimitDegrees=45
  swing1Motion=Free
  swing2LimitDegrees=45
  swing2Motion=Free
  twistLimitDegrees=45
  twistMotion=Free
 }
 {
  bone1Name=dyn_vest_side_r_04
  bone2Name=dyn_vest_bk_r_04
  swing1LimitDegrees=45
  swing1Motion=Free
  swing2LimitDegrees=45
  swing2Motion=Free
  twistLimitDegrees=45
  twistMotion=Free
 }
 {
  bone1Name=dyn_vest_bk_l_03
  bone2Name=dyn_vest_side_l_03
  swing1LimitDegrees=45
  swing1Motion=Free
  swing2LimitDegrees=45
  swing2Motion=Free
  twistLimitDegrees=45
  twistMotion=Free
 }
 {
  bone1Name=dyn_vest_bk_mid_03
  bone2Name=dyn_vest_bk_l_03
  swing1LimitDegrees=45
  swing1Motion=Free
  swing2LimitDegrees=45
  swing2Motion=Free
  twistLimitDegrees=45
  twistMotion=Free
 }
 {
  bone1Name=dyn_vest_bk_r_03
  bone2Name=dyn_vest_bk_mid_03
  swing1LimitDegrees=45
  swing1Motion=Free
  swing2LimitDegrees=45
  swing2Motion=Free
  twistLimitDegrees=45
  twistMotion=Free
 }
 {
  bone1Name=dyn_vest_side_r_03
  bone2Name=dyn_vest_bk_r_03
  swing1LimitDegrees=45
  swing1Motion=Free
  swing2LimitDegrees=45
  swing2Motion=Free
  twistLimitDegrees=45
  twistMotion=Free
 }
]
