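# Secondary-motion AnimBP for the female medium SparkArcher character body.
# Pipeline: parent-pose copy, Control Rig, four RigidBody sims, upper-arm correctives (active when physics is off), AnimDynamics laces, LOD fix, AnimDynamics straps, collar offset and weighted strap constraints.
# Event graph sets backbling, sliding and LOD state, plus gravity override for RigidBody nodes.
# It also drives material WPO cloth simulation on base and overlay materials, enabling wind and turbulence during skydiving or parachuting.
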AnimBlueprint /Game/Characters/Player/Female/Medium/Bodies/F_MED_SparkArcher/Meshes/F_MED_SparkArcher_AnimBP
target_skeleton=&/F_MED_SparkArcher_Skeleton
graphs [
 AnimGraph {nodes [
   n1 = "Transform (Modify) Bone - Bone: dyn_vest_collar_mid_base"!(
    ComponentPose=Local To Component(LocalPose=Use cached pose 'ToCollar'())
    bAlphaBoolEnabled:bool=True
    Translation="0,1,0"
    Rotation="0,0,0"
    Scale="1,1,1"
   ) -> (Pose) {
    alphaInputType=Bool
    translationMode=BMM_Additive
    alphaBoolBlend {blendOption=Linear}
    boneToModify {boneName=dyn_vest_collar_mid_base}
    translation {y=1}
   }
   n2 = "Constraint - dyn_upperarm_vest_strap_l_01 (PPPPPPPPP)"!(
    ComponentPose=@n1
    Alpha=1
    ConstraintWeights_0=0.0303
    ConstraintWeights_1:float=0.0171
    ConstraintWeights_2:float=0.0027
    ConstraintWeights_3:float=0.8797
    ConstraintWeights_4:float=0.004
    ConstraintWeights_5:float=0.0005
    ConstraintWeights_6:float=0.0492
    ConstraintWeights_7:float=0.0147
    ConstraintWeights_8:float=0.0019
   ) -> (Pose) {
    alphaBoolBlend {blendOption=Linear}
    boneToModify {boneName=dyn_upperarm_vest_strap_l_01}
    constraintSetup [
     {targetBone {boneName=spine_05}}
     {targetBone {boneName=clavicle_l}}
     {targetBone {boneName=upperarm_l}}
     {targetBone {boneName=upperarm_twist_01_l}}
     {targetBone {boneName=deform_upperarm_bk_base_l}}
     {targetBone {boneName=deform_upperarm_up_base_l}}
     {targetBone {boneName=deform_upperarm_up_l}}
     {targetBone {boneName=dyn_upperarm_vest_strap_l_01}}
     {targetBone {boneName=dyn_upperarm_vest_strap_l_02}}
    ]
    constraintWeights [1,1,1,1,1,1,1,1,1]
   }
   n3 = "Constraint - dyn_upperarm_vest_strap_r_01 (PPPPPPPPPP)"!(
    ComponentPose=@n2
    Alpha=1
    ConstraintWeights_0=0.0176
    ConstraintWeights_1:float=0.0939
    ConstraintWeights_2:float=0.1725
    ConstraintWeights_3:float=0.6513
    ConstraintWeights_4:float=0.0054
    ConstraintWeights_5:float=0.0007
    ConstraintWeights_6:float=0.0227
    ConstraintWeights_7:float=0.0354
    ConstraintWeights_8:float=0.0003
    ConstraintWeights_9:float=0.0004
   ) -> (Pose) {
    alphaBoolBlend {blendOption=Linear}
    boneToModify {boneName=dyn_upperarm_vest_strap_r_01}
    constraintSetup [
     {targetBone {boneName=spine_05}}
     {targetBone {boneName=clavicle_r}}
     {targetBone {boneName=upperarm_r}}
     {targetBone {boneName=upperarm_twist_01_r}}
     {targetBone {boneName=deform_upperarm_bk_r}}
     {targetBone {boneName=deform_upperarm_up_base_r}}
     {targetBone {boneName=deform_upperarm_up_r}}
     {targetBone {boneName=dyn_upperarm_vest_strap_r_01}}
     {targetBone {boneName=dyn_upperarm_vest_strap_r_02}}
     {targetBone {boneName=dyn_vest_collar_r_01}}
    ]
    constraintWeights [1,1,1,1,1,1,1,1,1,1]
   }
   n4 = "Constraint - dyn_vest_strap_frnt_l_01 (PPPPPPPPP)"!(
    ComponentPose=@n3
    Alpha=1
    ConstraintWeights_0=0.085
    ConstraintWeights_1:float=0.0855
    ConstraintWeights_2:float=0.0026
    ConstraintWeights_3:float=0.0091
    ConstraintWeights_4:float=0.1226
    ConstraintWeights_5:float=0.007
    ConstraintWeights_6:float=0.0247
    ConstraintWeights_7:float=0.0372
    ConstraintWeights_8:float=0.6264
   ) -> (Pose) {
    alphaBoolBlend {blendOption=Linear}
    boneToModify {boneName=dyn_vest_strap_frnt_l_01}
    constraintSetup [
     {targetBone {boneName=spine_05}}
     {targetBone {boneName=upperarm_twist_01_l}}
     {targetBone {boneName=deform_upperarm_fr_base_l}}
     {targetBone {boneName=deform_upperarm_fr_l}}
     {targetBone {boneName=dyn_vest_collar_frnt_l_01}}
     {targetBone {boneName=dyn_vest_collar_l_01}}
     {targetBone {boneName=dyn_vest_strap_frnt_l_02}}
     {targetBone {boneName=dyn_vest_strap_support_l}}
     {targetBone {boneName=dyn_vest_collar_frnt_l_support}}
    ]
    constraintWeights [1,1,1,1,1,1,1,1,1]
   }
   n5 = "Constraint - dyn_vest_strap_frnt_r_01 (PPPPPPPPPP)"!(
    ComponentPose=@n4
    Alpha=1
    ConstraintWeights_0=0.0331
    ConstraintWeights_1:float=0.0046
    ConstraintWeights_2:float=0.0306
    ConstraintWeights_3:float=0.0004
    ConstraintWeights_4:float=0.0246
    ConstraintWeights_5:float=0.021
    ConstraintWeights_6:float=0.0617
    ConstraintWeights_7:float=0.0031
    ConstraintWeights_8:float=0.0055
    ConstraintWeights_9:float=0.8154
   ) -> (Pose) {
    alphaBoolBlend {blendOption=Linear}
    boneToModify {boneName=dyn_vest_strap_frnt_r_01}
    constraintSetup [
     {targetBone {boneName=spine_05}}
     {targetBone {boneName=clavicle_r}}
     {targetBone {boneName=upperarm_twist_01_r}}
     {targetBone {boneName=deform_upperarm_fr_base_r}}
     {targetBone {boneName=deform_upperarm_fr_r}}
     {targetBone {boneName=dyn_vest_collar_frnt_r_01}}
     {targetBone {boneName=dyn_vest_strap_frnt_r_01}}
     {targetBone {boneName=dyn_vest_strap_frnt_r_02}}
     {targetBone {boneName=dyn_vest_strap_support_r}}
     {targetBone {boneName=dyn_vest_collar_frnt_r_support}}
    ]
    constraintWeights [1,1,1,1,1,1,1,1,1,1]
   }
   Output Pose(Result=Component To Local(ComponentPose=@n5)) {layerGroup=None}
   n6 = Control Rig!(
    Source=Copy Pose From Mesh() {bUseAttachedParent=true}
    Alpha
    RigidBody Enabled:bool=Get bRigidBodyEnabled()
    Is Tactical Sprinting:bool=Get bIsTacticalSprinting()
    Is All Sliding:bool=Get Is All Sliding() -> (Is All Sliding:bool)
   ) -> (Pose) {
    alphaBoolBlend {blendOption=Linear}
    controlRigAssetReference {blueprintRigClass=&/F_MED_SparkArcher_CtrlRig.F_MED_SparkArcher_CtrlRig_C}
   }
   Save cached pose 'ToRigidBody'(Pose=@n6) {cacheName=ToRigidBody}
   n7 = Anim Dynamics - Bone: dyn_strap_side_l_01(ComponentPose=Local To Component(LocalPose=Use cached pose 'AnimDynamics'()), Alpha=1, ExternalForce="0,0,0") {
    angularDampingOverride=1
    angularSpringConstant=100
    bAngularSpring=true
    bOverrideAngularDamping=true
    bUseGravityOverride=true
    simulationSpace=RootRelative
    alphaBoolBlend {blendOption=Linear}
    boundBone {boneName=dyn_strap_side_l_01}
    componentAppliedLinearAccClamp=<3000 3000 3000>
    componentLinearAccScale=<1 1 1>
    componentLinearVelScale=<1 1 1>
    gravityOverride {z=-980}
    physicsBodyDefinitions [{
      boundBone {boneName=dyn_strap_side_l_01}
      boxExtents=<2 2 2>
      constraintSetup {angularTargetAxis=AxisY, angularLimitsMax=<20 30 20>, angularLimitsMin=<-20 -30 -20>, angularTarget=<0 2 0>}
      localJointOffset=<0.05 0 0>
     }]
    relativeSpaceBone {boneName=pelvis}
   }
   n8 = Anim Dynamics - Bone: dyn_strap_side_l_02(ComponentPose=@n7, Alpha=1, ExternalForce="0,0,0") {
    angularDampingOverride=1
    angularSpringConstant=100
    bAngularSpring=true
    bOverrideAngularDamping=true
    bUseGravityOverride=true
    simulationSpace=RootRelative
    alphaBoolBlend {blendOption=Linear}
    boundBone {boneName=dyn_strap_side_l_02}
    componentAppliedLinearAccClamp=<3000 3000 3000>
    componentLinearAccScale=<1 1 1>
    componentLinearVelScale=<1 1 1>
    gravityOverride {z=-980}
    physicsBodyDefinitions [{
      boundBone {boneName=dyn_strap_side_l_02}
      boxExtents=<2 2 2>
      constraintSetup {angularTargetAxis=AxisY, angularLimitsMax=<20 30 20>, angularLimitsMin=<-20 -30 -20>, angularTarget=<0 2 0>}
      localJointOffset=<0.05 0 0>
     }]
    relativeSpaceBone {boneName=pelvis}
   }
   n9 = Anim Dynamics - Bone: dyn_strap_bk_l_01(ComponentPose=@n8, Alpha=1, ExternalForce="0,0,0") {
    angularDampingOverride=1
    angularSpringConstant=150
    bAngularSpring=true
    bOverrideAngularDamping=true
    bUseGravityOverride=true
    simulationSpace=RootRelative
    alphaBoolBlend {blendOption=Linear}
    boundBone {boneName=dyn_strap_bk_l_01}
    componentAppliedLinearAccClamp=<1000 1000 1000>
    componentLinearAccScale=<1 1 1>
    componentLinearVelScale=<1 1 1>
    gravityOverride {z=-980}
    physicsBodyDefinitions [{
      boundBone {boneName=dyn_strap_bk_l_01}
      boxExtents=<2 2 2>
      constraintSetup {angularTargetAxis=AxisY, angularLimitsMax=<20 20 20>, angularLimitsMin=<-20 -20 -10>, angularTarget=<0 2 0>}
      localJointOffset=<0.05 0 0>
     }]
    relativeSpaceBone {boneName=pelvis}
   }
   n10 = Anim Dynamics - Bone: dyn_strap_side_r_02(ComponentPose=@n9, Alpha=1, ExternalForce="0,0,0") {
    angularDampingOverride=1
    angularSpringConstant=100
    bAngularSpring=true
    bOverrideAngularDamping=true
    bUseGravityOverride=true
    simulationSpace=RootRelative
    alphaBoolBlend {blendOption=Linear}
    boundBone {boneName=dyn_strap_side_r_02}
    componentAppliedLinearAccClamp=<3000 3000 3000>
    componentLinearAccScale=<1 1 1>
    componentLinearVelScale=<1 1 1>
    gravityOverride {z=-980}
    physicsBodyDefinitions [{
      boundBone {boneName=dyn_strap_side_r_02}
      boxExtents=<2 2 2>
      constraintSetup {angularTargetAxis=AxisY, angularLimitsMax=<20 20 20>, angularLimitsMin=<-20 -20 -20>, angularTarget=<0 0 -0.5>}
      localJointOffset=<0.05 0 0>
     }]
    relativeSpaceBone {boneName=pelvis}
   }
   n11 = Anim Dynamics - Bone: dyn_strap_side_r_03(ComponentPose=@n10, Alpha=1, ExternalForce="0,0,0") {
    angularDampingOverride=1
    angularSpringConstant=100
    bAngularSpring=true
    bOverrideAngularDamping=true
    bUseGravityOverride=true
    simulationSpace=RootRelative
    alphaBoolBlend {blendOption=Linear}
    boundBone {boneName=dyn_strap_side_r_03}
    componentAppliedLinearAccClamp=<3000 3000 3000>
    componentLinearAccScale=<1 1 1>
    componentLinearVelScale=<1 1 1>
    gravityOverride {z=-980}
    physicsBodyDefinitions [{
      boundBone {boneName=dyn_strap_side_r_03}
      boxExtents=<2 2 2>
      constraintSetup {angularTargetAxis=AxisY, twistAxis=AxisY, angularLimitsMax=<20 20 20>, angularLimitsMin=<-20 -20 -20>, angularTarget=<0 0 -0.5>}
      localJointOffset=<0.05 0 0>
     }]
    relativeSpaceBone {boneName=pelvis}
   }
   n12 = Anim Dynamics - Bone: dyn_strap_bk_r_01(ComponentPose=@n11, Alpha=1, ExternalForce="0,0,0") {
    angularDampingOverride=1
    angularSpringConstant=250
    bAngularSpring=true
    bOverrideAngularDamping=true
    bUseGravityOverride=true
    simulationSpace=RootRelative
    alphaBoolBlend {blendOption=Linear}
    boundBone {boneName=dyn_strap_bk_r_01}
    componentAppliedLinearAccClamp=<1000 1000 1000>
    componentLinearAccScale=<1 1 1>
    componentLinearVelScale=<1 1 1>
    gravityOverride {z=-980}
    physicsBodyDefinitions [{
      boundBone {boneName=dyn_strap_bk_r_01}
      boxExtents=<2 2 2>
      constraintSetup {angularTargetAxis=AxisY, angularLimitsMax=<20 20 20>, angularLimitsMin=<-20 -20 -20>, angularTarget=<0 0 0>}
      localJointOffset=<0.05 0 0>
     }]
    relativeSpaceBone {boneName=pelvis}
   }
   n13 = Anim Dynamics - Bone: dyn_strap_bk_r_02(ComponentPose=@n12, Alpha=1, ExternalForce="0,0,0") {
    angularDampingOverride=1
    angularSpringConstant=250
    bAngularSpring=true
    bOverrideAngularDamping=true
    bUseGravityOverride=true
    simulationSpace=RootRelative
    alphaBoolBlend {blendOption=Linear}
    boundBone {boneName=dyn_strap_bk_r_02}
    componentAppliedLinearAccClamp=<1000 1000 1000>
    componentLinearAccScale=<1 1 1>
    componentLinearVelScale=<1 1 1>
    gravityOverride {z=-980}
    physicsBodyDefinitions [{
      boundBone {boneName=dyn_strap_bk_r_02}
      boxExtents=<2 2 2>
      constraintSetup {angularTargetAxis=AxisY, angularLimitsMax=<20 20 20>, angularLimitsMin=<-20 -20 -20>, angularTarget=<0 0 0>}
      localJointOffset=<0.05 0 0>
     }]
    relativeSpaceBone {boneName=pelvis}
   }
   n14 = Anim Dynamics - Bone: dyn_strap_side_r_01(ComponentPose=@n13, Alpha=1, ExternalForce="0,0,0") {
    angularDampingOverride=0.9999
    angularSpringConstant=150
    bAngularSpring=true
    bOverrideAngularDamping=true
    bUseGravityOverride=true
    simulationSpace=RootRelative
    alphaBoolBlend {blendOption=Linear}
    boundBone {boneName=dyn_strap_side_r_01}
    componentAppliedLinearAccClamp=<1000 1000 1000>
    componentLinearAccScale=<1 1 1>
    componentLinearVelScale=<1 1 1>
    gravityOverride {z=-980}
    physicsBodyDefinitions [{
      boundBone {boneName=dyn_strap_side_r_01}
      boxExtents=<2 2 2>
      constraintSetup {angularTargetAxis=AxisY, angularLimitsMax=<20 20 20>, angularLimitsMin=<-20 -20 -20>, angularTarget=<0 0 -0.5>}
      localJointOffset=<0.1 0 0>
     }]
    relativeSpaceBone {boneName=pelvis}
   }
   Save cached pose 'ToCollar'(Pose=Component To Local(ComponentPose=@n14)) {cacheName=ToCollar}
   n15 = Get Gravity Override()
   n16 = RigidBody!(ComponentPose=Local To Component(LocalPose=Use cached pose 'ToRigidBody'()), Alpha, OverrideWorldGravity:Vector=@n15, ExternalForce) -> (Pose) {
    bClampLinearTranslationLimitToRefPose=true
    bOverrideWorldGravity=true
    overridePhysicsAsset=&/F_MED_SparkArcher_Physics
    simulationSpace=BaseBoneSpace
    alphaBoolBlend {blendOption=Linear}
    baseBoneRef {boneName=pelvis}
    componentAppliedLinearAccClamp=<1000 1000 1000>
    componentLinearAccScale=<0.5 0.5 0.5>
    componentLinearVelScale=<0.5 0.5 0.5>
    overrideWorldGravity {z=-980}
    simSpaceSettings {maxAngularAcceleration=200, maxAngularVelocity=0, maxLinearAcceleration=100, maxLinearVelocity=50, velocityScaleZ=0.2, worldAlpha=1}
   }
   n17 = RigidBody(ComponentPose=@n16, OverrideWorldGravity=@n15) {
    bClampLinearTranslationLimitToRefPose=true
    bOverrideWorldGravity=true
    overridePhysicsAsset=&/F_MED_SparkArcher_ArmLStrap_Physics
    simulationSpace=BaseBoneSpace
    alphaBoolBlend {blendOption=Linear}
    baseBoneRef {boneName=pelvis}
    componentAppliedLinearAccClamp=<3000 3000 3000>
    overrideWorldGravity {z=-980}
    simSpaceSettings {maxAngularAcceleration=500, maxAngularVelocity=0, maxLinearAcceleration=400, maxLinearVelocity=250, velocityScaleZ=0.2, worldAlpha=1}
   }
   n18 = RigidBody(ComponentPose=@n17, OverrideWorldGravity=@n15) {
    bClampLinearTranslationLimitToRefPose=true
    bOverrideWorldGravity=true
    overridePhysicsAsset=&/F_MED_SparkArcher_ArmRStrap_Physics
    simulationSpace=BaseBoneSpace
    alphaBoolBlend {blendOption=Linear}
    baseBoneRef {boneName=pelvis}
    componentAppliedLinearAccClamp=<3000 3000 3000>
    overrideWorldGravity {z=-980}
    simSpaceSettings {maxAngularAcceleration=500, maxAngularVelocity=0, maxLinearAcceleration=400, maxLinearVelocity=250, velocityScaleZ=0.2, worldAlpha=1}
   }
   n19 = RigidBody(ComponentPose=@n18, OverrideWorldGravity=@n15) {
    bClampLinearTranslationLimitToRefPose=true
    bOverrideWorldGravity=true
    overridePhysicsAsset=&/F_MED_SparkArcher_Collar_Physics
    simulationSpace=BaseBoneSpace
    alphaBoolBlend {blendOption=Linear}
    baseBoneRef {boneName=spine_05}
    componentAppliedLinearAccClamp=<3000 3000 3000>
    overrideWorldGravity {z=-980}
    simSpaceSettings {maxAngularAcceleration=500, maxAngularVelocity=0, maxLinearAcceleration=400, maxLinearVelocity=250, velocityScaleZ=0.2, worldAlpha=1}
   }
   n20 = Get IsRigidBodyOff() -> (IsRigidBodyOff:bool)
   n21 = "Transform (Modify) Bone - Bone: dyn_upperarm_ALT_correct_l"!(ComponentPose=@n19, bAlphaBoolEnabled:bool=@n20, Translation="0,0,0", Rotation="0,0,0", Scale="0.6,0.2,0.7") -> (Pose) {
    alphaInputType=Bool
    rotationSpace=BCS_BoneSpace
    scaleMode=BMM_Additive
    scaleSpace=BCS_BoneSpace
    translationSpace=BCS_BoneSpace
    alphaBoolBlend {blendOption=Linear}
    boneToModify {boneName=dyn_upperarm_ALT_correct_l}
    scale=<0.6 0.2 0.7>
   }
   n22 = "Transform (Modify) Bone - Bone: dyn_upperarm_ALT_correct_r"!(ComponentPose=@n21, bAlphaBoolEnabled:bool=@n20, Translation="-4,-4,4", Rotation="0,0,0", Scale="0.6,0.2,0.7") -> (Pose) {
    alphaInputType=Bool
    rotationSpace=BCS_BoneSpace
    scaleMode=BMM_Additive
    scaleSpace=BCS_BoneSpace
    translationMode=BMM_Additive
    translationSpace=BCS_BoneSpace
    alphaBoolBlend {blendOption=Linear}
    boneToModify {boneName=dyn_upperarm_ALT_correct_r}
    scale=<0.6 0.2 0.7>
    translation=<-4 -4 4>
   }
   Save cached pose 'RigidBody'(Pose=Component To Local(ComponentPose=@n22)) {cacheName=RigidBody}
   n23 = Anim Dynamics - Bone: dyn_foot_lace_l_01(ComponentPose=Local To Component(LocalPose=Use cached pose 'RigidBody'()), Alpha=1, ExternalForce="0,0,0") {
    angularDampingOverride=0.9999
    angularSpringConstant=250
    bAngularSpring=true
    bOverrideAngularDamping=true
    bUseGravityOverride=true
    simulationSpace=RootRelative
    alphaBoolBlend {blendOption=Linear}
    boundBone {boneName=dyn_foot_lace_l_01}
    componentAppliedLinearAccClamp=<1000 1000 1000>
    componentLinearAccScale=<1 1 1>
    componentLinearVelScale=<1 1 1>
    gravityOverride {z=-920}
    physicsBodyDefinitions [{
      boundBone {boneName=dyn_foot_lace_l_01}
      boxExtents=<2 2 2>
      constraintSetup {angularLimitsMax=<0 30 20>, angularLimitsMin=<0 -30 -3>, angularTarget=<0 2 0>}
      localJointOffset=<0.1 0 0>
     }]
    relativeSpaceBone {boneName=calf_twist_02_r}
   }
   n24 = Anim Dynamics - Bone: dyn_foot_lace_l_02(ComponentPose=@n23, Alpha=1, ExternalForce="0,0,0") {
    angularDampingOverride=0.9999
    angularSpringConstant=250
    bAngularSpring=true
    bOverrideAngularDamping=true
    bUseGravityOverride=true
    simulationSpace=RootRelative
    alphaBoolBlend {blendOption=Linear}
    boundBone {boneName=dyn_foot_lace_l_02}
    componentAppliedLinearAccClamp=<1000 1000 1000>
    componentLinearAccScale=<1 1 1>
    componentLinearVelScale=<1 1 1>
    gravityOverride {z=-920}
    physicsBodyDefinitions [{
      boundBone {boneName=dyn_foot_lace_l_02}
      boxExtents=<2 2 2>
      constraintSetup {angularLimitsMax=<0 30 20>, angularLimitsMin=<0 -30 -3>, angularTarget=<0 2 0>}
      localJointOffset=<0.1 0 0>
     }]
    relativeSpaceBone {boneName=calf_twist_02_r}
   }
   n25 = Anim Dynamics - Bone: dyn_foot_lace_r_02(ComponentPose=@n24, Alpha=1, ExternalForce="0,0,0") {
    angularDampingOverride=0.9999
    angularSpringConstant=250
    bAngularSpring=true
    bOverrideAngularDamping=true
    bUseGravityOverride=true
    simulationSpace=RootRelative
    alphaBoolBlend {blendOption=Linear}
    boundBone {boneName=dyn_foot_lace_r_02}
    componentAppliedLinearAccClamp=<1000 1000 1000>
    componentLinearAccScale=<1 1 1>
    componentLinearVelScale=<1 1 1>
    gravityOverride {z=-920}
    physicsBodyDefinitions [{
      boundBone {boneName=dyn_foot_lace_r_02}
      boxExtents=<2 2 2>
      constraintSetup {angularLimitsMax=<30 5 30>, angularLimitsMin=<-5 -20 -5>, angularTarget=<0 -1 0>}
      localJointOffset=<0.1 0 0>
     }]
    relativeSpaceBone {boneName=calf_twist_02_r}
   }
   n26 = Anim Dynamics - Bone: dyn_foot_lace_r_01(ComponentPose=@n25, Alpha=1, ExternalForce="0,0,0") {
    angularDampingOverride=0.9999
    angularSpringConstant=250
    bAngularSpring=true
    bOverrideAngularDamping=true
    bUseGravityOverride=true
    simulationSpace=RootRelative
    alphaBoolBlend {blendOption=Linear}
    boundBone {boneName=dyn_foot_lace_r_01}
    componentAppliedLinearAccClamp=<1000 1000 1000>
    componentLinearAccScale=<1 1 1>
    componentLinearVelScale=<1 1 1>
    gravityOverride {z=-920}
    physicsBodyDefinitions [{
      boundBone {boneName=dyn_foot_lace_r_01}
      boxExtents=<2 2 2>
      constraintSetup {angularLimitsMax=<5 5 10>, angularLimitsMin=<-20 -20 -20>, angularTarget=<0 0 -0.5>}
      localJointOffset=<0.1 0 0>
     }]
    relativeSpaceBone {boneName=calf_twist_02_r}
   }
   n27 = "Transform (Modify) Bone - Bone: dyn_top_correct_LOD1"(ComponentPose=@n26, Alpha=1, Translation="2,0,0", Rotation="0,0,0", Scale="1,1,1") {
    translationMode=BMM_Replace
    translationSpace=BCS_BoneSpace
    alphaBoolBlend {blendOption=Linear}
    boneToModify {boneName=dyn_top_correct_LOD1}
    translation {x=2}
   }
   n28 = "Transform (Modify) Bone - Bone: dyn_top_correct_LOD1"(ComponentPose=@n27, Alpha=Get Is LOD1() -> (Is LOD1:double), Translation="2,0,0", Rotation="0,0,0", Scale="1,1,1") {
    translationMode=BMM_Replace
    translationSpace=BCS_BoneSpace
    alphaBoolBlend {blendOption=Linear}
    boneToModify {boneName=dyn_top_correct_LOD1}
    translation {x=2}
   }
   Save cached pose 'AnimDynamics'(Pose=Component To Local(ComponentPose=@n28)) {cacheName=AnimDynamics}
  ]}
 EventGraph {nodes [
   n1 = Event BlueprintInitializeAnimation()
   n2 = Sequence!(execute=[@n1.then, @n19.then]) -> (then_0, then_1, then_2:Exec)
   n3 = CheckForBackbling(LongBackblingThreshold=40)
   > Branch(Condition=@n3)
   > n4 = Sequence()
   > Set IsBackblingOn(IsBackblingOn:bool=@n3) -> (Output_Get:bool)
   n5 = Branch(execute=@n4.then_1, Condition=OR Boolean(A=@n3.bIsBackblingACape, B=@n3.bIsBackblingLong))
   > Set IsLongBackblingOn(IsLongBackblingOn:bool=true) -> (Output_Get:bool)
   > Set IsShortBackblingOn(IsShortBackblingOn:bool=false) -> (Output_Get:bool)
   Set IsLongBackblingOn(execute=@n5.else, IsLongBackblingOn=false)
   > Set IsShortBackblingOn(IsShortBackblingOn=true)
   n6 = GetPartSkeletalMeshComponent()
   n7 = Set Mesh(execute=@n2.then_1, Mesh:Skeletal Mesh Component Object Reference=@n6) -> (Output_Get:Skeletal Mesh Component Object Reference)
   > n8 = Get Dynamic Material Instance(SkeletalMeshComponent=@n7)
   > Set Material Instance Dynamic(Material Instance Dynamic:Material Instance Dynamic Object Reference=@n8) -> (Output_Get:Material Instance Dynamic Object Reference)
   > n9 = GetOrCreateDynamicOverlayMaterial(MeshComponent=@n6, MaterialName=None)
   > Set Overlay Mat Instance Dyanmic(Overlay Mat Instance Dyanmic:Material Instance Dynamic Object Reference=@n9) -> (Output_Get:Material Instance Dynamic Object Reference)
   n10 = CheckForBackbling(LongBackblingThreshold=80)
   Set Is backpack(execute=@n2.then_2, Is backpack:bool=AND Boolean(A=@n10, B=NOT Boolean(A=@n10.bIsBackblingLong))) -> (Output_Get:bool)
   > Set Is NoBackpack(Is NoBackpack:bool=AND Boolean!(A=NOT Boolean(A=@n10), B=NOT Boolean(A=@n10.bIsBackblingACape), C:bool=NOT Boolean(A=@n10.bIsBackblingLong)) -> (ReturnValue)) -> (Output_Get:bool)
   n11 = Event BlueprintUpdateAnimation()
   > n12 = Sequence()
   n13 = Gravity Override State Function DEPRECATED(
    execute=@n12.then_1
    CharacterPartAnimInstance=Self-Reference()
    GameTimeInSeconds=GetGameTimeInSeconds()
    GravityOverrideParams=Get Gravity Override Params 0() -> (Gravity Override Params 0:Gravity Override Params Struct Structure)
   )
   > Set Gravity Override(Gravity Override=@n13)
   n14 = Get Mesh!() -> (Mesh)
   n15 = Get Overlay Mat Instance Dyanmic() -> (Overlay Mat Instance Dyanmic:Material Instance Dynamic Object Reference)
   n16 = Get Material Instance Dynamic() -> (Material Instance Dynamic:Material Instance Dynamic Object Reference)
   > Set WPO Globals Function(
    Material=@n16
    Overlay Material=@n15
    Skeletal Mesh=@n14
    "Local Noise Transform  Location"="Get Local Noise Transform  Location"() -> ("Local Noise Transform  Location":Vector)
    Local Noise Offset=Get Local Noise Offset()
   )
   > Set WPO Regional Function!(
    execute
    Material:Material Instance Dynamic Object Reference=@n16
    Overlay Material:Material Instance Dynamic Object Reference=@n15
    Skeletal Mesh:Skeletal Mesh Component Object Reference=@n14
    Velocity Transform:Name=pelvis
    Velocity Interp Speed:double=Get Velocity Interp Speed()
    Collision Transform:Name=pelvis
    Collision Transform Axis:EAxis Enum
    Invert Collision Transform Axis:bool=false
    Delta Time X:double=@n11.DeltaTimeX
    Previous Velocity Transform Location:"Vector (by ref)"=Get R1 Velocity Transform Location()
    Previous Velocity Transform Velocity:"Vector (by ref)"=Get R1 Velocity Transform Velocity()
    Region:int=1
   ) -> (then)
   > Set WPO Regional Function(
    Material=@n16
    Overlay Material=@n15
    Skeletal Mesh=@n14
    Velocity Transform=spine_2
    Velocity Interp Speed=Get Velocity Interp Speed()
    Collision Transform=spine_2
    Invert Collision Transform Axis=true
    Delta Time X=@n11.DeltaTimeX
    Previous Velocity Transform Location=Get L Previous Velocity Transform Location() -> (L Previous Velocity Transform Location:Vector)
    Previous Velocity Transform Velocity=Get L Previous Velocity Transform Velocity() -> (L Previous Velocity Transform Velocity:Vector)
    Region=2
   )
   > SetScalarParameterValue(self=[@n16, @n15], ParameterName=ClothSim_WindStrength, Value=Get WPO Wind() -> (WPO Wind:double))
   > SetScalarParameterValue(self=[@n16, @n15], ParameterName=ClothSim_TurbulenceNoiseTileDivisor, Value=Get Turbulence Noise() -> (Turbulence Noise:float))
   n17 = Branch(
    execute=@n12.then_2
    Condition=OR Boolean!(A=Get bIsSkydiveFloating(), B=Get bIsSkydiveDiveMode(), C:bool=Get bIsParachuteOpened()) -> (ReturnValue)
   )
   > Set Velocity Interp Speed(Velocity Interp Speed:double=10) -> (Output_Get:double)
   > Set WPO Wind(WPO Wind:double=1) -> (Output_Get:double)
   > Set Turbulence Noise(Turbulence Noise:float=200) -> (Output_Get:float)
   Set Velocity Interp Speed(execute=@n17.else, Velocity Interp Speed=0)
   > Set WPO Wind(WPO Wind=0)
   > Set Turbulence Noise(Turbulence Noise=0)
   Set IsRigidBodyOff(execute=@n12.then_0, IsRigidBodyOff:bool=NOT Boolean(A=Get bRigidBodyEnabled())) -> (Output_Get:bool)
   n18 = Get CurrentLOD()
   > Set Is LOD1(Is LOD1:double="To Float (Boolean)"(InBool="integer > integer"(A=@n18))) -> (Output_Get:double)
   > Set Is All Sliding(Is All Sliding:bool=OR Boolean!(A=Get bIsSliding(), B=Get bIsBaseballSliding(), C=Get bIsKneeSliding(), D:bool=Get bIsSlopeSliding() -> (bIsSlopeSliding:bool)) -> (ReturnValue)) -> (Output_Get:bool)
   n19 = Event OnCharacterCustomizationLoaded()
   AND Boolean!(A=@n10, B=@n10.bIsBackblingLong) -> (ReturnValue)
  ]}
]
variables [
 Gravity Override {type=Vector}
 Gravity Override Params 0 {type=GravityOverrideParamsStruct}
 Is All Sliding {type=boolean}
 Is backpack {type=boolean}
 Is LOD1 {type=number}
 Is NoBackpack {type=boolean}
 IsBackblingOn {type=boolean}
 IsLongBackblingOn {type=boolean}
 IsRigidBodyOff {type=boolean}
 IsShortBackblingOn {type=boolean}
 L Previous Velocity Transform Location {type=Vector}
 L Previous Velocity Transform Velocity {type=Vector}
 Lace Deform {type=number}
 Local Noise Offset {type=Vector}
 "Local Noise Transform  Location" {type=Vector}
 Material Instance Dynamic {type=MaterialInstanceDynamic}
 Mesh {type=SkeletalMeshComponent}
 Overlay Mat Instance Dyanmic {type=MaterialInstanceDynamic}
 R1 Velocity Transform Location {type=Vector}
 R1 Velocity Transform Velocity {type=Vector}
 Turbulence Noise {type=number}
 Velocity Interp Speed {type=number}
 WPO Wind {type=number}
]
