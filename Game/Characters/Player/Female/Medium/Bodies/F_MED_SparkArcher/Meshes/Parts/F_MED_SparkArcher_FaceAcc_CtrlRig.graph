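# Control Rig for the SparkArcher face accessory mesh that applies corrective offsets to dynamic hair base bones as the body poses.
# With RigidBody Enabled off, head rotation relative to spine_05 alone drives a back-hair correction.
# When on, a sequence layers side and back hair corrections, weighted by spherical pose readers on clavicles and pelvis plus remapped head rotation.
# Some reader weights are multiplied to require combined poses, and one is Kalman-smoothed.
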ControlRigBlueprint /Game/Characters/Player/Female/Medium/Bodies/F_MED_SparkArcher/Meshes/Parts/F_MED_SparkArcher_FaceAcc_CtrlRig
graphs [
 RigVMFunctionLibrary {links [], nodes []}
 RigVMFunction_Sequence_1_ContainedGraph {nodes [
   n1 = RigVMFunctionEntryNode()
   RigVMFunction_Sequence = &RigVMFunction_Sequence(ExecuteContext=@n1.ExecuteContext)
   n2 = &RigVMFunction_Sequence(ExecuteContext=@RigVMFunction_Sequence.B)
   RigVMFunctionReturnNode(A=@RigVMFunction_Sequence.A, B=@n2.A, C=@n2.B)
  ]}
 RigVMModel {nodes [
   RigUnit_BeginExecution = &RigUnit_BeginExecution()
   RigVMFunction_ControlFlowBranch = &RigVMFunction_ControlFlowBranch(ExecuteContext=@RigUnit_BeginExecution.ExecutePin, Condition=RigVMVariableNode() {variable=RigidBody Enabled}.Value)
   RigUnit_GetRelativeTransformForItem = &RigUnit_GetRelativeTransformForItem(Child=(Type=Bone,Name="head"), Parent=(Type=Bone,Name="spine_05"), bChildInitial=False, bParentInitial=False)
   RigVMFunction_MathQuaternionToEuler = &RigVMFunction_MathQuaternionToEuler(RotationOrder=ZYX, Value=@RigUnit_GetRelativeTransformForItem.RelativeTransform.Rotation)
   Remap = &RigVMFunction_MathDoubleRemap(SourceMaximum=-49, SourceMinimum=-25, TargetMaximum=1, TargetMinimum=0, Value=@RigVMFunction_MathQuaternionToEuler.Result.Z)
   &RigUnit_ModifyTransforms(
    ItemToModify="((Item=(Type=Bone,Name=\"dyn_hair_bk_r_base\"),Transform=(Rotation=(X=0,Y=-0.4715,Z=0,W=0.8819),Translation=(X=0,Y=0,Z=0),Scale3D=(X=1,Y=1,Z=1))),(Item=(Type=Bone,Name=\"dyn_hair_bk_l_base\"),Transform=(Rotation=(X=-0.0428,Y=-0.2828,Z=-0.4456,W=0.8483),Transla ...(truncated)"
    WeightMinimum=0
    Weight=@Remap.Result
    ExecutePin=@RigVMFunction_ControlFlowBranch.False
   )
   RigVMFunction_Sequence = RigVMAggregateNode(ExecuteContext=@RigVMFunction_ControlFlowBranch.True)
   n1 = &RigUnit_SphericalPoseReader(
    ActiveRegionScaleFactors=(PositiveWidth=1,NegativeWidth=1,PositiveHeight=1,NegativeHeight=1)
    ActiveRegionSize=0.04
    Debug=(bDrawDebug=true,bDraw2D=False,bDrawLocalAxes=False,DebugScale=25,DebugSegments=20,DebugThickness=0.25)
    DriverAxis=(X=1,Y=0,Z=0)
    DriverItem=(Type=Bone,Name="clavicle_l")
    FalloffRegionScaleFactors=(PositiveWidth=0.4,NegativeWidth=0.4,PositiveHeight=0.5,NegativeHeight=0.5)
    FalloffSize=0.15
    FlipHeightScaling=False
    FlipWidthScaling=False
    OptionalParentItem=(Type=Bone,Name="None")
    RotationOffset=(X=0,Y=-60,Z=-10)
    ExecutePin=@RigVMFunction_Sequence.A
   )
   > n2 = &RigUnit_SphericalPoseReader(
    ActiveRegionScaleFactors=(PositiveWidth=1,NegativeWidth=1,PositiveHeight=1,NegativeHeight=1)
    ActiveRegionSize=0.04
    Debug=(bDrawDebug=true,bDraw2D=False,bDrawLocalAxes=False,DebugScale=25,DebugSegments=20,DebugThickness=0.25)
    DriverAxis=(X=-1,Y=0,Z=0)
    DriverItem=(Type=Bone,Name="clavicle_r")
    FalloffRegionScaleFactors=(PositiveWidth=0.4,NegativeWidth=0.4,PositiveHeight=0.5,NegativeHeight=0.5)
    FalloffSize=0.15
    FlipHeightScaling=False
    FlipWidthScaling=False
    OptionalParentItem=(Type=Bone,Name="None")
    RotationOffset=(X=0,Y=110,Z=0)
   )
   KalmanFilter = &RigVMFunction_KalmanFloat(BufferSize=7, Value=&RigVMFunction_MathDoubleMul(A=@n1.OutputParam, B=@n2.OutputParam).Result) {template="KalmanFilter::Execute(in Value,in BufferSize,out Result)"}
   > &RigUnit_ModifyTransforms(
    ItemToModify="((Item=(Type=Bone,Name=\"dyn_hair_side_r_base\"),Transform=(Rotation=(X=0,Y=0,Z=-0.3301,W=0.9439),Translation=(X=-1.01,Y=-2.1705,Z=0),Scale3D=(X=1,Y=1,Z=1))),(Item=(Type=Bone,Name=\"dyn_hair_side_l_base\"),Transform=(Rotation=(X=0,Y=0,Z=0,W=1),Translation=(X=0 ...(truncated)"
    WeightMinimum=0
    Weight=@KalmanFilter.Result
   )
   n3 = &RigUnit_SphericalPoseReader(
    ActiveRegionScaleFactors=(PositiveWidth=1,NegativeWidth=1,PositiveHeight=1,NegativeHeight=1)
    ActiveRegionSize=0.03
    Debug=(bDrawDebug=true,bDraw2D=False,bDrawLocalAxes=False,DebugScale=25,DebugSegments=20,DebugThickness=0.25)
    DriverAxis=(X=-1,Y=0,Z=0)
    DriverItem=(Type=Bone,Name="clavicle_r")
    FalloffRegionScaleFactors=(PositiveWidth=1,NegativeWidth=0.7,PositiveHeight=1,NegativeHeight=1)
    FalloffSize=0.3
    FlipHeightScaling=False
    FlipWidthScaling=False
    OptionalParentItem=(Type=Bone,Name="None")
    RotationOffset=(X=-15,Y=110,Z=0)
    ExecutePin=@RigVMFunction_Sequence.B
   )
   > &RigUnit_ModifyTransforms(
    ItemToModify="((Item=(Type=Bone,Name=\"None\"),Transform=(Rotation=(X=0,Y=0,Z=0,W=1),Translation=(X=0,Y=0,Z=0),Scale3D=(X=1,Y=1,Z=1))),(Item=(Type=Bone,Name=\"dyn_hair_side_r_base\"),Transform=(Rotation=(X=0,Y=0,Z=0,W=1),Translation=(X=-2,Y=1,Z=-2),Scale3D=(X=1,Y=1,Z=1))),( ...(truncated)"
    WeightMinimum=0
    ItemToModify.1.Transform.Rotation=&RigVMFunction_MathQuaternionFromEuler(Euler=(X=0,Y=20,Z=30)).Result
    ItemToModify.2.Transform.Rotation=&RigVMFunction_MathQuaternionFromEuler(Euler=(X=0,Y=20,Z=30)).Result
    Weight=@n3.OutputParam
   )
   > n4 = &RigUnit_SphericalPoseReader(
    ActiveRegionScaleFactors=(PositiveWidth=1,NegativeWidth=1,PositiveHeight=1,NegativeHeight=1)
    ActiveRegionSize=0.05
    Debug=(bDrawDebug=true,bDraw2D=False,bDrawLocalAxes=False,DebugScale=25,DebugSegments=20,DebugThickness=0.25)
    DriverAxis=(X=1,Y=0,Z=0)
    DriverItem=(Type=Bone,Name="clavicle_l")
    FalloffRegionScaleFactors=(PositiveWidth=1,NegativeWidth=1,PositiveHeight=1,NegativeHeight=1)
    FalloffSize=0.2
    FlipHeightScaling=False
    FlipWidthScaling=False
    OptionalParentItem=(Type=Bone,Name="None")
    RotationOffset=(X=0,Y=-45,Z=20)
   )
   > &RigUnit_ModifyTransforms(
    ItemToModify="((Item=(Type=Bone,Name=\"None\"),Transform=(Rotation=(X=0,Y=0,Z=0,W=1),Translation=(X=0,Y=0,Z=0),Scale3D=(X=1,Y=1,Z=1))),(Item=(Type=Bone,Name=\"dyn_hair_side_l_base\"),Transform=(Rotation=(X=0,Y=0,Z=0,W=1),Translation=(X=1,Y=0,Z=0),Scale3D=(X=1,Y=1,Z=1))),(It ...(truncated)"
    WeightMinimum=0
    ItemToModify.2.Transform.Rotation=&RigVMFunction_MathQuaternionFromEuler(Euler=(X=0,Y=-10,Z=0)).Result
    ItemToModify.1.Transform.Rotation=&RigVMFunction_MathQuaternionFromEuler(Euler=(X=0,Y=0,Z=40)).Result
    Weight=@n4.OutputParam
   )
   > n5 = &RigUnit_SphericalPoseReader(
    ActiveRegionScaleFactors=(PositiveWidth=1,NegativeWidth=1,PositiveHeight=1,NegativeHeight=1)
    ActiveRegionSize=0.05
    Debug=(bDrawDebug=false,bDraw2D=False,bDrawLocalAxes=False,DebugScale=25,DebugSegments=20,DebugThickness=0.25)
    DriverAxis=(X=-1,Y=0,Z=0)
    DriverItem=(Type=Bone,Name="clavicle_r")
    FalloffRegionScaleFactors=(PositiveWidth=1,NegativeWidth=1,PositiveHeight=1,NegativeHeight=0.5)
    FalloffSize=0.2
    FlipHeightScaling=False
    FlipWidthScaling=False
    OptionalParentItem=(Type=Bone,Name="spine_05")
    RotationOffset=(X=135,Y=-3,Z=90)
   )
   > n6 = &RigUnit_SphericalPoseReader(
    ActiveRegionScaleFactors=(PositiveWidth=1,NegativeWidth=1,PositiveHeight=1,NegativeHeight=1)
    ActiveRegionSize=0.05
    Debug=(bDrawDebug=false,bDraw2D=False,bDrawLocalAxes=False,DebugScale=25,DebugSegments=20,DebugThickness=0.25)
    DriverAxis=(X=1,Y=0,Z=0)
    DriverItem=(Type=Bone,Name="clavicle_l")
    FalloffRegionScaleFactors=(PositiveWidth=1,NegativeWidth=1,PositiveHeight=1,NegativeHeight=0.5)
    FalloffSize=0.2
    FlipHeightScaling=False
    FlipWidthScaling=False
    OptionalParentItem=(Type=Bone,Name="spine_05")
    RotationOffset=(X=-40,Y=-3,Z=90)
   )
   Multiply = &RigVMFunction_MathDoubleMul(A=@n6.OutputParam, B=@n5.OutputParam)
   > &RigUnit_ModifyTransforms(
    ItemToModify="((Item=(Type=Bone,Name=\"dyn_hair_bk_r_base\"),Transform=(Rotation=(X=-0.2484,Y=-0.3258,Z=0.0166,W=0.9121),Translation=(X=0,Y=0,Z=0),Scale3D=(X=1,Y=1,Z=1))),(Item=(Type=Bone,Name=\"dyn_hair_bk_l_base\"),Transform=(Rotation=(X=-0.3213,Y=-0.1135,Z=-0.4278,W=0.83 ...(truncated)"
    WeightMinimum=0
    Weight=@Multiply.Result
   )
   n7 = &RigUnit_SphericalPoseReader(
    ActiveRegionScaleFactors=(PositiveWidth=1,NegativeWidth=1,PositiveHeight=1,NegativeHeight=1)
    ActiveRegionSize=0.1
    Debug=(bDrawDebug=false,bDraw2D=False,bDrawLocalAxes=False,DebugScale=25,DebugSegments=20,DebugThickness=0.25)
    DriverAxis=(X=1,Y=0,Z=0)
    DriverItem=(Type=Bone,Name="clavicle_l")
    FalloffRegionScaleFactors=(PositiveWidth=1,NegativeWidth=1,PositiveHeight=1,NegativeHeight=1)
    FalloffSize=0.15
    FlipHeightScaling=False
    FlipWidthScaling=False
    OptionalParentItem=(Type=Bone,Name="None")
    RotationOffset=(X=0,Y=-70,Z=-30)
    ExecutePin=@RigVMFunction_Sequence.C
   )
   > n8 = &RigUnit_SphericalPoseReader(
    ActiveRegionScaleFactors=(PositiveWidth=1,NegativeWidth=1,PositiveHeight=1,NegativeHeight=1)
    ActiveRegionSize=0.1
    Debug=(bDrawDebug=false,bDraw2D=False,bDrawLocalAxes=False,DebugScale=25,DebugSegments=20,DebugThickness=0.25)
    DriverAxis=(X=0,Y=1,Z=0)
    DriverItem=(Type=Bone,Name="pelvis")
    FalloffRegionScaleFactors=(PositiveWidth=1,NegativeWidth=1,PositiveHeight=1,NegativeHeight=1)
    FalloffSize=0.2
    FlipHeightScaling=False
    FlipWidthScaling=False
    OptionalParentItem=(Type=Bone,Name="root")
    RotationOffset=(X=0,Y=90,Z=10)
   )
   > &RigUnit_ModifyTransforms(
    ItemToModify="((Item=(Type=Bone,Name=\"dyn_hair_side_r_base\"),Transform=(Rotation=(X=0,Y=0,Z=0,W=1),Translation=(X=0,Y=0,Z=0),Scale3D=(X=1,Y=1,Z=1))),(Item=(Type=Bone,Name=\"dyn_hair_side_l_base\"),Transform=(Rotation=(X=0,Y=0,Z=0,W=1),Translation=(X=0,Y=0,Z=0),Scale3D=(X= ...(truncated)"
    WeightMinimum=0
    Weight=&RigVMFunction_MathDoubleMul(B=@n7.OutputParam, A=@n8.OutputParam).Result
    ItemToModify.1.Transform.Rotation=&RigVMFunction_MathQuaternionFromEuler(Euler=(X=0,Y=50,Z=-30)).Result
    ItemToModify.0.Transform.Rotation=&RigVMFunction_MathQuaternionFromEuler(Euler=(X=0,Y=30,Z=-20)).Result
    ItemToModify.2.Transform.Rotation=&RigVMFunction_MathQuaternionFromEuler(Euler=(X=35,Y=-15,Z=-35)).Result
   )
   n9 = &RigVMFunction_MathQuaternionToEuler(
    RotationOrder=ZYX
    Value=&RigUnit_GetRelativeTransformForItem(Child=(Type=Bone,Name="head"), Parent=(Type=Bone,Name="spine_05"), bChildInitial=False, bParentInitial=False).RelativeTransform.Rotation
   )
   > &RigUnit_ModifyTransforms(
    ItemToModify="((Item=(Type=Bone,Name=\"dyn_hair_bk_r_base\"),Transform=(Rotation=(X=-0.2117,Y=-0.409,Z=-0.0472,W=0.8864),Translation=(X=0,Y=0,Z=0),Scale3D=(X=1,Y=1,Z=1))),(Item=(Type=Bone,Name=\"dyn_hair_bk_l_base\"),Transform=(Rotation=(X=-0.3219,Y=-0.1146,Z=-0.4274,W=0.83 ...(truncated)"
    WeightMinimum=0
    Weight=&RigVMFunction_MathDoubleRemap(SourceMaximum=-71, SourceMinimum=-60, TargetMaximum=1, TargetMinimum=0, Value=@n9.Result.Z).Result
   )
   > &RigUnit_ModifyTransforms(
    ItemToModify="((Item=(Type=Bone,Name=\"dyn_hair_side_r_base\"),Transform=(Rotation=(X=0,Y=0,Z=-0.1764,W=0.9843),Translation=(X=0,Y=0,Z=0),Scale3D=(X=1,Y=1,Z=1))),(Item=(Type=Bone,Name=\"dyn_hair_side_l_base\"),Transform=(Rotation=(X=0,Y=0,Z=-0.1581,W=0.9874),Translation=(X= ...(truncated)"
    WeightMinimum=0
    Weight=&RigVMFunction_MathDoubleRemap(SourceMaximum=-30, SourceMinimum=-10, TargetMaximum=1, TargetMinimum=0, Value=@n9.Result.Z).Result
   )
   &RigUnit_ModifyTransforms(
    ExecutePin=()
    ItemToModify=((Item=(Type=Bone,Name="pelvis"),Transform=(Rotation=(X=0,Y=0,Z=0.1752,W=0.9845),Translation=(X=0,Y=0,Z=0),Scale3D=(X=1,Y=1,Z=1))))
    Weight=1
    WeightMinimum=0
   )
   &RigUnit_ModifyTransforms(
    ExecutePin=()
    ItemToModify=((Item=(Type=Bone,Name="clavicle_l"),Transform=(Rotation=(X=0,Y=-0.0604,Z=0,W=0.9982),Translation=(X=0,Y=0,Z=0),Scale3D=(X=1,Y=1,Z=1))))
    Weight=1
    WeightMinimum=0
   )
   &RigVMFunction_Sequence(ExecuteContext=())
  ]}
]
hierarchy [
 root {loc=<0 0 0>, rot=<0 0 0 1>}
 pelvis {parent=root, loc=<0 -0.5301 98.0571>, rot=<0.0127 0.707 -0.0127 -0.707>}
 spine_01 {parent=pelvis, loc=<6.1208 -0.9328 0>, rot=<0 0 0.034 -0.9994>}
 spine_02 {parent=spine_01, loc=<6.5996 -0.2095 0>, rot=<0 0 0.0238 -0.9997>}
 spine_03 {parent=spine_02, loc=<6.4914 -0.1237 0>, rot=<0 0 0.0149 -0.9999>}
 spine_04 {parent=spine_03, loc=<6.9598 0.0794 0>, rot=<0 0 -0.0132 -0.9999>}
 spine_05 {parent=spine_04, loc=<6.9987 0 0>, rot=<0 0 -0.0015 -1>}
 neck_01 {parent=spine_05, loc=<11.3138 0.1529 0>, rot=<0 0 -0.2003 -0.9797>}
 neck_02 {parent=neck_01, loc=<4.5627 0 0>, rot=<0 0 0.0504 -0.9987>}
 head {parent=neck_02, loc=<5.3914 0 0>, rot=<0 0 0.115 -0.9934>}
 dyn_earring_r_01 {parent=head, loc=<-0.3817 0.6123 6.5522>, rot=<0.0024 -0.9878 -0.155 0.0153>}
 dyn_earring_r_02 {parent=dyn_earring_r_01, loc=<0.935 0.0505 0.0088>, rot=<-0.2316 0 0 -0.9728>}
 dyn_earring_l_01 {parent=head, loc=<-0.4372 0.6854 -6.4432>, rot=<0.0093 -0.9794 0.1943 -0.0536>}
 dyn_bow_r_01 {parent=head, loc=<5.1957 -1.4639 8.3745>, rot=<0.1149 0.5392 0.1488 -0.8209>}
 dyn_bow_r_02 {parent=dyn_bow_r_01, loc=<0.7703 0.5708 0.0298>, rot=<-0.3161 -0.5983 0.04 -0.7352>}
 dyn_bow_r_03 {parent=dyn_bow_r_01, loc=<0.5668 -0.6048 0.0918>, rot=<0.5292 -0.3645 0.5859 -0.4937>}
 dyn_beanie_l_01 {parent=head, loc=<16.6418 -3.4427 -1.6078>, rot=<-0.4447 0.2835 0.242 -0.8144>}
 dyn_beanie_l_02 {parent=head, loc=<13.3521 -6.6151 -1.7024>, rot=<-0.4877 0.0933 0.1359 -0.8573>}
 dyn_beanie_r_01 {parent=head, loc=<16.7653 -3.4497 3.2287>, rot=<0.785 -0.2817 -0.357 0.4207>}
 dyn_beanie_r_02 {parent=head, loc=<13.4116 -6.534 3.1573>, rot=<0.2244 -0.1305 0.2337 -0.937>}
 dyn_goggles_01 {parent=head, loc=<14.3532 8.0924 0.0109>, rot=<0.8883 0.4592 0 0>}
 dyn_hair_side_l_base {parent=head, loc=<0.5521 -3.4441 -5.5184>, rot=<0.0225 0.0807 -0.1001 -0.9914>}
 dyn_hair_side_l_01 {parent=dyn_hair_side_l_base, loc=<0 0 0>, rot=<0 0 0 1>}
 dyn_hair_side_l_02 {parent=dyn_hair_side_l_01, loc=<-4.275 1.4224 -0.0434>, rot=<-0.1443 0.0014 0.0812 -0.9862>}
 dyn_hair_side_l_03 {parent=dyn_hair_side_l_02, loc=<3.7925 1.0174 -0.1362>, rot=<0.0044 -0.0244 -0.176 -0.9841>}
 dyn_hair_side_l_06 {parent=dyn_hair_side_l_01, loc=<1.4755 -0.349 1.6445>, rot=<0.2075 0.253 -0.2492 -0.9115>}
 dyn_hair_side_l_04 {parent=dyn_hair_side_l_01, loc=<-3.5339 -1.297 1.5484>, rot=<0.4296 0.0897 -0.0573 -0.8967>}
 dyn_hair_side_l_05 {parent=dyn_hair_side_l_04, loc=<3.0761 -0.0087 -0.1211>, rot=<0.0116 -0.1146 0.0069 -0.9933>}
 dyn_hair_side_r_base {parent=head, loc=<0.246 -3.5389 6.0636>, rot=<0.156 -0.0148 0.9824 -0.1015>}
 dyn_hair_side_r_01 {parent=dyn_hair_side_r_base, loc=<0 0 0>, rot=<0 0 0 1>}
 dyn_hair_side_r_02 {parent=dyn_hair_side_r_01, loc=<4.2747 -1.4223 0.0435>, rot=<-0.2811 -0.0166 0.1796 -0.9426>}
 dyn_hair_side_r_03 {parent=dyn_hair_side_r_02, loc=<3.8005 -0.6214 0.0836>, rot=<-0.0118 -0.0662 0.1757 -0.9821>}
 dyn_hair_side_r_07 {parent=dyn_hair_side_r_01, loc=<1.0357 -1.1253 1.4546>, rot=<0.0641 0.1753 0.0542 -0.9809>}
 dyn_hair_side_r_04 {parent=dyn_hair_side_r_01, loc=<3.3504 1.5034 -1.0509>, rot=<0.1402 0.1513 -0.0119 -0.9784>}
 dyn_hair_side_r_05 {parent=dyn_hair_side_r_04, loc=<3.7929 -0.4918 -0.4638>, rot=<0.0192 -0.1887 0.0992 -0.9768>}
 dyn_hair_side_r_06 {parent=dyn_hair_side_r_04, loc=<2.2065 1.1036 0.7068>, rot=<0.2483 0.2703 0.0723 -0.9274>}
 dyn_hair_bk_l_base {parent=head, loc=<-0.8965 -6.1956 -2.1445>, rot=<0.0724 -0.8108 0.572 -0.1012>}
 dyn_hair_bk_l_01 {parent=dyn_hair_bk_l_base, loc=<0 0 0>, rot=<0 0 0 1>}
 dyn_hair_bk_l_02 {parent=dyn_hair_bk_l_01, loc=<4.4735 0 0.542>, rot=<0 0.1686 0 -0.9857>}
 dyn_hair_bk_r_base {parent=head, loc=<-0.896 -6.1956 2.1445>, rot=<0.1012 -0.572 0.8108 -0.0724>}
 dyn_hair_bk_r_01 {parent=dyn_hair_bk_r_base, loc=<0 0 0>, rot=<0 0 0 1>}
 dyn_hair_bk_r_02 {parent=dyn_hair_bk_r_01, loc=<4.4738 0 0.5419>, rot=<0 0.1686 0 -0.9857>}
 dyn_beanie {parent=head, loc=<15.8626 -4.8787 0.1945>, rot=<-0.6613 0.2435 0.2304 -0.671>}
 FACIAL_C_FacialRoot {parent=head, loc=<1.4649 0.658 -0.0007>, rot=<0.5021 -0.4979 -0.5021 -0.4979>}
 FACIAL_C_Nose {parent=FACIAL_C_FacialRoot, loc=<-0.0007 0.2548 8.8007>, rot=<-0.1305 0 0 -0.9914>}
 FACIAL_L_Nostril {parent=FACIAL_C_Nose, loc=<1.0296 2.1988 -0.1544>, rot=<0.0173 0.1305 0.0023 -0.9913>}
 dyn_hair_mid_correct {parent=head, loc=<-2.6025 -6.2722 0.0205>, rot=<0.0737 -0.7021 0.7042 -0.0756>}
 dyn_hair_bk_l_correct {parent=head, loc=<-2.162 -4.637 -4.8645>, rot=<0.0598 -0.9451 0.302 -0.1095>}
 dyn_hair_bk_r_correct {parent=head, loc=<-2.162 -4.637 4.8645>, rot=<0.9451 0.0598 -0.1095 -0.302>}
 clavicle_l {parent=spine_05, loc=<5.4831 -0.1053 -3.2946>, rot=<0.0465 -0.7289 -0.009 -0.6829>}
 clavicle_r {parent=spine_05, loc=<5.4831 -0.1053 3.2946>, rot=<0.7289 0.0465 -0.6829 0.009>}
 dyn_vest_collar_mid {parent=spine_05, loc=<11.5649 -12.6857 0>, rot=<0 0 0.0149 -0.9999>}
]
variables [RigidBody Enabled {type=bool, default=False}]
