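# Physics asset for the SparkArcher female medium face accessory mesh, defining bodies for dynamic hair.
# Capsule bodies form left and right back and side hair groups, each rooted in a tiny base body.
# Back chains run two segments from the base, while side chains fork from the first segment into two branches, all with tight swing and twist limits.
# Larger unconstrained neck and upper-spine capsules provide body shapes, likely for hair collision.
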PhysicsAsset /Game/Characters/Player/Female/Medium/Bodies/F_MED_SparkArcher/Meshes/Parts/F_MED_SparkArcher_FaceAcc_Physics
bodies {
 dyn_hair_bk_l_01 [{shapeType=Capsule, length=4, radius=2, shapeName=dyn_hair_bk_l_01_capsule, center=<2.4511 0 0>, rotation=<rot 90 0 0>}]
 dyn_hair_bk_l_02 [{shapeType=Capsule, length=6.0294, radius=2, shapeName=dyn_hair_bk_l_02_capsule, center=<2.922 0 0>, rotation=<rot 90 -90 -0.3137>}]
 dyn_hair_bk_l_base [{shapeType=Capsule, length=0.505, radius=0.505, shapeName=dyn_hair_bk_l_base_capsule}]
 dyn_hair_bk_r_01 [{shapeType=Capsule, length=4, radius=2, shapeName=dyn_hair_bk_r_01_capsule, center=<2.4511 0 0>, rotation=<rot 90 0 0>}]
 dyn_hair_bk_r_02 [{shapeType=Capsule, length=6.0294, radius=2, shapeName=dyn_hair_bk_r_02_capsule, center=<2.922 0 0>, rotation=<rot 90 11.0996 100.786>}]
 dyn_hair_bk_r_base [{shapeType=Capsule, length=0.505, radius=0.505, shapeName=dyn_hair_bk_r_base_capsule}]
 dyn_hair_side_l_01 [{shapeType=Capsule, length=4, radius=2, shapeName=dyn_hair_side_l_01_capsule, center=<-2.2397 0 0>, rotation=<rot 90 0 0>}]
 dyn_hair_side_l_02 [{
   shapeType=Capsule
   length=5.7267
   radius=2
   shapeName=dyn_hair_side_l_02_capsule
   center=<-4.0129 1.0478 -0.5009>
   rotation=<rot 83.462 -21.2737 9.2351>
  }]
 dyn_hair_side_l_04 [{
   shapeType=Capsule
   length=4.4003
   radius=1.5997
   shapeName=dyn_hair_side_l_04_capsule
   center=<-3.0551 0 -0.4656>
   rotation=<rot 90 0.0006 90>
  }]
 dyn_hair_side_l_base [{shapeType=Capsule, length=0.505, radius=0.505, shapeName=dyn_hair_side_l_base_capsule}]
 dyn_hair_side_r_01 [{shapeType=Capsule, length=4, radius=2, shapeName=dyn_hair_side_r_01_capsule, center=<2.4511 0 0>, rotation=<rot 90 0 0>}]
 dyn_hair_side_r_02 [{
   shapeType=Capsule
   length=5.1328
   radius=1.8893
   shapeName=dyn_hair_side_r_02_capsule
   center=<3.7161 -0.7803 -1.0367>
   rotation=<rot 79.5535 11.4804 26.0106>
  }]
 dyn_hair_side_r_04 [{
   shapeType=Capsule
   length=4
   radius=2
   shapeName=dyn_hair_side_r_04_capsule
   center=<2.8259 0 0.4431>
   rotation=<rot 77.7652 179.9998 179.9999>
  }]
 dyn_hair_side_r_base [{shapeType=Capsule, length=0.505, radius=0.505, shapeName=dyn_hair_side_r_base_capsule}]
 neck_01 [{shapeType=Capsule, length=18.0927, radius=6.2768, shapeName=neck_01_capsule, center=<0 1.8609 0>, rotation=<rot -90 -166.7373 166.7373>}]
 spine_05 [{shapeType=Capsule, length=10.2741, radius=10.9574, shapeName=spine_05_capsule, center=<7.686 4.3348 0>}]
}
constraints [
 {bone1Name=dyn_hair_bk_l_01, bone2Name=dyn_hair_bk_l_base, swing1LimitDegrees=10, swing2LimitDegrees=20, twistLimitDegrees=3}
 {bone1Name=dyn_hair_bk_l_02, bone2Name=dyn_hair_bk_l_01, swing1LimitDegrees=10, swing2LimitDegrees=10, twistLimitDegrees=3}
 {bone1Name=dyn_hair_bk_r_01, bone2Name=dyn_hair_bk_r_base, swing1LimitDegrees=10, swing2LimitDegrees=20, twistLimitDegrees=1}
 {bone1Name=dyn_hair_bk_r_02, bone2Name=dyn_hair_bk_r_01, swing1LimitDegrees=20, swing2LimitDegrees=20, twistLimitDegrees=1}
 {bone1Name=dyn_hair_side_l_01, bone2Name=dyn_hair_side_l_base, swing1LimitDegrees=10, swing2LimitDegrees=10, twistLimitDegrees=3}
 {bone1Name=dyn_hair_side_l_02, bone2Name=dyn_hair_side_l_01, swing1LimitDegrees=10, swing2LimitDegrees=10, twistLimitDegrees=3}
 {bone1Name=dyn_hair_side_l_04, bone2Name=dyn_hair_side_l_01, swing1LimitDegrees=10, swing2LimitDegrees=10, twistLimitDegrees=3}
 {bone1Name=dyn_hair_side_r_02, bone2Name=dyn_hair_side_r_01, swing1LimitDegrees=20, swing2LimitDegrees=20, twistLimitDegrees=1}
 {bone1Name=dyn_hair_side_r_01, bone2Name=dyn_hair_side_r_base, swing1LimitDegrees=10, swing2LimitDegrees=10, twistLimitDegrees=1}
 {bone1Name=dyn_hair_side_r_04, bone2Name=dyn_hair_side_r_01, swing1LimitDegrees=20, swing2LimitDegrees=20, twistLimitDegrees=1}
]
